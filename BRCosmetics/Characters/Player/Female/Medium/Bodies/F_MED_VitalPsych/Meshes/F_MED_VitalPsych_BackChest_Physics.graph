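# Physics asset for the back-chest straps on the VitalPsych medium female player cosmetic body, giving the left and right straps capsule bodies for physics-driven motion.
# Each side chains a base bone and two dyn bones from a spine_05 anchor capsule.
# Base-to-spine constraints leave swing and twist free, while the dyn joints use narrow twist and moderate swing limits.
# The right-side dyn capsules are fatter and offset farther than the left.
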PhysicsAsset /BRCosmetics/Characters/Player/Female/Medium/Bodies/F_MED_VitalPsych/Meshes/F_MED_VitalPsych_BackChest_Physics
bodies {
 base_back_chest_strap_00_l [{shapeType=Capsule, length=0.505, radius=0.505, shapeName=base_back_chest_strap_00_l_capsule, rotation=<rot 0 -90 0>}]
 base_back_chest_strap_00_r [{shapeType=Capsule, length=0.505, radius=0.505, shapeName=base_back_chest_strap_00_r_capsule, rotation=<rot 0 -90 0>}]
 dyn_back_chest_strap_00_l [{shapeType=Capsule, length=1, radius=1, shapeName=dyn_back_chest_strap_00_l_capsule, center=<1 0 0>, rotation=<rot 90 0 0>}]
 dyn_back_chest_strap_00_r [{shapeType=Capsule, length=0.5, radius=1.5, shapeName=dyn_back_chest_strap_00_r_capsule, center=<1.5875 0 0>, rotation=<rot 90 0 0>}]
 dyn_back_chest_strap_01_l [{shapeType=Capsule, length=1, radius=1, shapeName=dyn_back_chest_strap_01_l_capsule, center=<2 0 0>, rotation=<rot 90 0 0>}]
 dyn_back_chest_strap_01_r [{shapeType=Capsule, length=0.75, radius=1.5, shapeName=dyn_back_chest_strap_01_r_capsule, center=<3.0586 0 0>, rotation=<rot 90 0 0>}]
 spine_05 [{shapeType=Capsule, length=2, radius=2, shapeName=spine_05_capsule}]
}
constraints [
 {
  bone1Name=base_back_chest_strap_00_l
  bone2Name=spine_05
  swing1LimitDegrees=45
  swing1Motion=Free
  swing2LimitDegrees=45
  swing2Motion=Free
  twistLimitDegrees=45
  twistMotion=Free
 }
 {bone1Name=dyn_back_chest_strap_00_l, bone2Name=base_back_chest_strap_00_l, swing1LimitDegrees=20, swing2LimitDegrees=45, twistLimitDegrees=5}
 {bone1Name=dyn_back_chest_strap_01_l, bone2Name=dyn_back_chest_strap_00_l, swing1LimitDegrees=20, swing2LimitDegrees=20, twistLimitDegrees=5}
 {
  bone1Name=base_back_chest_strap_00_r
  bone2Name=spine_05
  swing1LimitDegrees=45
  swing1Motion=Free
  swing2LimitDegrees=45
  swing2Motion=Free
  twistLimitDegrees=45
  twistMotion=Free
 }
 {bone1Name=dyn_back_chest_strap_00_r, bone2Name=base_back_chest_strap_00_r, swing1LimitDegrees=20, swing2LimitDegrees=45, twistLimitDegrees=5}
 {bone1Name=dyn_back_chest_strap_01_r, bone2Name=dyn_back_chest_strap_00_r, swing1LimitDegrees=20, swing2LimitDegrees=20, twistLimitDegrees=5}
]
