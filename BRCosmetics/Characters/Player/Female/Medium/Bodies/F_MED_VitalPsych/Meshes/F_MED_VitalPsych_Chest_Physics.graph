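# Physics asset for the VitalPsych female medium body's chest mesh, set up around its left and right front chest straps.
# Each strap is a two-link dyn_ capsule chain on a base bone, with limited swing and tight twist.
# Free-rotation constraints tie strap bases and upper arms to the clavicles, and those to spine_04.
# Upper arms carry the largest capsules; left and right strap capsules differ in size.
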PhysicsAsset /BRCosmetics/Characters/Player/Female/Medium/Bodies/F_MED_VitalPsych/Meshes/F_MED_VitalPsych_Chest_Physics
bodies {
 base_front_chest_strap_00_l [{shapeType=Capsule, length=0.505, radius=0.505, shapeName=base_front_chest_strap_00_l_capsule, rotation=<rot 0 -90 0>}]
 base_front_chest_strap_00_r [{shapeType=Capsule, length=0.505, radius=0.505, shapeName=base_front_chest_strap_00_r_capsule, rotation=<rot 0 -90 0>}]
 clavicle_l [{shapeType=Capsule, length=1, radius=1, shapeName=clavicle_l_capsule}]
 clavicle_r [{shapeType=Capsule, length=1, radius=1, shapeName=clavicle_r_capsule}]
 dyn_front_chest_strap_00_l [{shapeType=Capsule, length=2.5, radius=1, shapeName=dyn_front_chest_strap_00_l_capsule, center=<2.2516 0 0>, rotation=<rot 90 0 0>}]
 dyn_front_chest_strap_00_r [{shapeType=Capsule, length=0.25, radius=1.5, shapeName=dyn_front_chest_strap_00_r_capsule, center=<2 0 0>, rotation=<rot 90 0 0>}]
 dyn_front_chest_strap_01_l [{shapeType=Capsule, length=2.75, radius=1, shapeName=dyn_front_chest_strap_01_l_capsule, center=<3.1129 0 0>, rotation=<rot 90 0 0>}]
 dyn_front_chest_strap_01_r [{shapeType=Capsule, length=2.25, radius=1.25, shapeName=dyn_front_chest_strap_01_r_capsule, center=<3.239 0 0>, rotation=<rot 90 0 0>}]
 spine_04 [{shapeType=Capsule, length=1, radius=1, shapeName=spine_04_capsule}]
 upperarm_l [{shapeType=Capsule, length=14, radius=4, shapeName=upperarm_l_capsule, center=<12 0 0>, rotation=<rot 90 0 0>}]
 upperarm_r [{shapeType=Capsule, length=14, radius=4, shapeName=upperarm_r_capsule, center=<-12 0 0>, rotation=<rot 90 0 0>}]
}
constraints [
 {
  bone1Name=dyn_front_chest_strap_00_r
  bone2Name=base_front_chest_strap_00_r
  swing1LimitDegrees=45
  swing2LimitDegrees=45
  twistLimitDegrees=5
 }
 {
  bone1Name=dyn_front_chest_strap_01_r
  bone2Name=dyn_front_chest_strap_00_r
  swing1LimitDegrees=45
  swing2LimitDegrees=45
  twistLimitDegrees=5
 }
 {
  bone1Name=dyn_front_chest_strap_00_l
  bone2Name=base_front_chest_strap_00_l
  swing1LimitDegrees=45
  swing2LimitDegrees=45
  twistLimitDegrees=5
 }
 {
  bone1Name=dyn_front_chest_strap_01_l
  bone2Name=dyn_front_chest_strap_00_l
  swing1LimitDegrees=45
  swing2LimitDegrees=45
  twistLimitDegrees=5
 }
 {
  bone1Name=upperarm_l
  bone2Name=clavicle_l
  swing1LimitDegrees=45
  swing1Motion=Free
  swing2LimitDegrees=45
  swing2Motion=Free
  twistLimitDegrees=45
  twistMotion=Free
 }
 {
  bone1Name=base_front_chest_strap_00_l
  bone2Name=clavicle_l
  swing1LimitDegrees=45
  swing1Motion=Free
  swing2LimitDegrees=45
  swing2Motion=Free
  twistLimitDegrees=45
  twistMotion=Free
 }
 {
  bone1Name=upperarm_r
  bone2Name=clavicle_r
  swing1LimitDegrees=45
  swing1Motion=Free
  swing2LimitDegrees=45
  swing2Motion=Free
  twistLimitDegrees=45
  twistMotion=Free
 }
 {
  bone1Name=base_front_chest_strap_00_r
  bone2Name=clavicle_r
  swing1LimitDegrees=45
  swing1Motion=Free
  swing2LimitDegrees=45
  swing2Motion=Free
  twistLimitDegrees=45
  twistMotion=Free
 }
 {
  bone1Name=clavicle_l
  bone2Name=spine_04
  swing1LimitDegrees=45
  swing1Motion=Free
  swing2LimitDegrees=45
  swing2Motion=Free
  twistLimitDegrees=45
  twistMotion=Free
 }
 {
  bone1Name=clavicle_r
  bone2Name=spine_04
  swing1LimitDegrees=45
  swing1Motion=Free
  swing2LimitDegrees=45
  swing2Motion=Free
  twistLimitDegrees=45
  twistMotion=Free
 }
]
