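# Physics asset for the F_MED_VitalPsych female medium player body, covering only its front and back right-thigh straps and a pocket.
# Strap chains hang from thigh_twist_01_r and fork into a/b ends (back after two links, front after one), while the pocket hangs from the pelvis.
# Small base_ capsules join parents with free angular motion.
# Larger dyn_ capsules have limited swing and narrow twist, tighter on strap ends and pocket.
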PhysicsAsset /BRCosmetics/Characters/Player/Female/Medium/Bodies/F_MED_VitalPsych/Meshes/F_MED_VitalPsych_Physics
bodies {
 base_back_leg_strap_00 [{shapeType=Capsule, length=0.505, radius=0.505, shapeName=base_back_leg_strap_00_capsule, rotation=<rot 0 -90 0>}]
 base_back_leg_strap_a_00 [{shapeType=Capsule, length=0.505, radius=0.505, shapeName=base_back_leg_strap_a_00_capsule, rotation=<rot 0 -90 0>}]
 base_back_leg_strap_b_00 [{shapeType=Capsule, length=0.505, radius=0.505, shapeName=base_back_leg_strap_b_00_capsule, rotation=<rot 0 -90 0>}]
 base_front_leg_strap_00 [{shapeType=Capsule, length=0.505, radius=0.505, shapeName=base_front_leg_strap_00_capsule, rotation=<rot 0 -90 0>}]
 base_front_leg_strap_a_00 [{shapeType=Capsule, length=0.505, radius=0.505, shapeName=base_front_leg_strap_a_00_capsule, rotation=<rot 0 -90 0>}]
 base_front_leg_strap_b_00 [{shapeType=Capsule, length=0.505, radius=0.505, shapeName=base_front_leg_strap_b_00_capsule, rotation=<rot 0 -90 0>}]
 base_pocket [{shapeType=Capsule, length=0.505, radius=0.505, shapeName=base_pocket_capsule, rotation=<rot 0 -90 0>}]
 dyn_back_leg_strap_00 [{shapeType=Capsule, length=2, radius=2.25, shapeName=dyn_back_leg_strap_00_capsule, center=<5.0359 0 0>, rotation=<rot 90 0 0>}]
 dyn_back_leg_strap_01 [{shapeType=Capsule, length=2, radius=2.25, shapeName=dyn_back_leg_strap_01_capsule, center=<5.0359 0 0>, rotation=<rot 90 0 0>}]
 dyn_back_leg_strap_a_00 [{shapeType=Capsule, length=4.75, radius=1.5, shapeName=dyn_back_leg_strap_a_00_capsule, center=<7.3837 0 0>, rotation=<rot 90 0 0>}]
 dyn_back_leg_strap_b_00 [{shapeType=Capsule, length=4.75, radius=1.5, shapeName=dyn_back_leg_strap_b_00_capsule, center=<7.3837 0 0>, rotation=<rot 90 0 0>}]
 dyn_front_leg_strap_00 [{shapeType=Capsule, length=2, radius=2.25, shapeName=dyn_front_leg_strap_00_capsule, center=<5.0359 0 0>, rotation=<rot 90 0 0>}]
 dyn_front_leg_strap_a_00 [{shapeType=Capsule, length=4.75, radius=1.5, shapeName=dyn_front_leg_strap_a_00_capsule, center=<7.3837 0 0>, rotation=<rot 90 0 0>}]
 dyn_front_leg_strap_b_00 [{shapeType=Capsule, length=4.75, radius=1.5, shapeName=dyn_front_leg_strap_b_00_capsule, center=<7.3837 0 0>, rotation=<rot 90 0 0>}]
 dyn_pocket [{shapeType=Capsule, length=3.5, radius=3.25, shapeName=dyn_pocket_capsule, center=<4.9788 0 0>, rotation=<rot 90 0 0>}]
 pelvis [{shapeType=Capsule, length=1, radius=1, shapeName=pelvis_capsule}]
 thigh_twist_01_r [{shapeType=Capsule, length=1, radius=1, shapeName=thigh_twist_01_r_capsule}]
}
constraints [
 {bone1Name=dyn_back_leg_strap_00, bone2Name=base_back_leg_strap_00, swing1LimitDegrees=45, swing2LimitDegrees=45, twistLimitDegrees=1}
 {bone1Name=dyn_back_leg_strap_01, bone2Name=dyn_back_leg_strap_00, swing1LimitDegrees=45, swing2LimitDegrees=45, twistLimitDegrees=1}
 {
  bone1Name=base_back_leg_strap_a_00
  bone2Name=dyn_back_leg_strap_01
  swing1LimitDegrees=45
  swing1Motion=Free
  swing2LimitDegrees=45
  swing2Motion=Free
  twistLimitDegrees=45
  twistMotion=Free
 }
 {bone1Name=dyn_back_leg_strap_a_00, bone2Name=base_back_leg_strap_a_00, swing1LimitDegrees=15, swing2LimitDegrees=15, twistLimitDegrees=1}
 {
  bone1Name=base_back_leg_strap_b_00
  bone2Name=dyn_back_leg_strap_01
  swing1LimitDegrees=45
  swing1Motion=Free
  swing2LimitDegrees=45
  swing2Motion=Free
  twistLimitDegrees=45
  twistMotion=Free
 }
 {bone1Name=dyn_back_leg_strap_b_00, bone2Name=base_back_leg_strap_b_00, swing1LimitDegrees=15, swing2LimitDegrees=15, twistLimitDegrees=1}
 {bone1Name=dyn_front_leg_strap_00, bone2Name=base_front_leg_strap_00, swing1LimitDegrees=45, swing2LimitDegrees=45, twistLimitDegrees=1}
 {
  bone1Name=base_front_leg_strap_a_00
  bone2Name=dyn_front_leg_strap_00
  swing1LimitDegrees=45
  swing1Motion=Free
  swing2LimitDegrees=45
  swing2Motion=Free
  twistLimitDegrees=45
  twistMotion=Free
 }
 {bone1Name=dyn_front_leg_strap_a_00, bone2Name=base_front_leg_strap_a_00, swing1LimitDegrees=15, swing2LimitDegrees=15, twistLimitDegrees=1}
 {
  bone1Name=base_front_leg_strap_b_00
  bone2Name=dyn_front_leg_strap_00
  swing1LimitDegrees=45
  swing1Motion=Free
  swing2LimitDegrees=45
  swing2Motion=Free
  twistLimitDegrees=45
  twistMotion=Free
 }
 {bone1Name=dyn_front_leg_strap_b_00, bone2Name=base_front_leg_strap_b_00, swing1LimitDegrees=15, swing2LimitDegrees=15, twistLimitDegrees=1}
 {bone1Name=dyn_pocket, bone2Name=base_pocket, swing1LimitDegrees=10, swing2LimitDegrees=5, twistLimitDegrees=5}
 {
  bone1Name=base_front_leg_strap_00
  bone2Name=thigh_twist_01_r
  swing1LimitDegrees=45
  swing1Motion=Free
  swing2LimitDegrees=45
  swing2Motion=Free
  twistLimitDegrees=45
  twistMotion=Free
 }
 {
  bone1Name=base_back_leg_strap_00
  bone2Name=thigh_twist_01_r
  swing1LimitDegrees=45
  swing1Motion=Free
  swing2LimitDegrees=45
  swing2Motion=Free
  twistLimitDegrees=45
  twistMotion=Free
 }
 {
  bone1Name=base_pocket
  bone2Name=pelvis
  swing1LimitDegrees=45
  swing1Motion=Free
  swing2LimitDegrees=45
  swing2Motion=Free
  twistLimitDegrees=45
  twistMotion=Free
 }
 {
  bone1Name=thigh_twist_01_r
  bone2Name=pelvis
  swing1LimitDegrees=45
  swing1Motion=Free
  swing2LimitDegrees=45
  swing2Motion=Free
  twistLimitDegrees=45
  twistMotion=Free
 }
]
