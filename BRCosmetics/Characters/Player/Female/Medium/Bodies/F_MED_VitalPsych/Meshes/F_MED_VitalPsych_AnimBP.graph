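# Post-process-style Anim Blueprint for the VitalPsych female cosmetic body, copying the parent mesh pose and adding deform, neck collar, strap and attachment motion.
# Cached stages chain Control Rig and bone constraints into collar correctives, then RigidBody straps, with Trail controllers when rigid bodies are disabled.
# The progressive variant selects hood-up or hood-down, switching collar, physics and backpack/cape offsets.
# EventGraph derives blend alphas from movement states and head-bone deltas.
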AnimBlueprint /BRCosmetics/Characters/Player/Female/Medium/Bodies/F_MED_VitalPsych/Meshes/F_MED_VitalPsych_AnimBP
target_skeleton=&/F_MED_VitalPsych_Skeleton
graphs [
 AnimGraph {nodes [
   n1 = "Transform (Modify) Bone - Bone: attach_backpack"!(
    ComponentPose=Local To Component(LocalPose=Use cached pose 'StrapsCache'())
    bAlphaBoolEnabled:bool=Get HoodDownMode() -> (HoodDownMode:bool)
    Translation="0,-4.5,0"
    Rotation="0,0,-5"
    Scale="1,1,1"
   ) -> (Pose) {
    alphaInputType=Bool
    rotationMode=BMM_Additive
    rotationSpace=BCS_BoneSpace
    translationMode=BMM_Additive
    translationSpace=BCS_BoneSpace
    alphaBoolBlend {blendOption=Linear}
    boneToModify {boneName=attach_backpack}
    rotation {roll=-5}
    translation {y=-4.5}
   }
   n2 = "Transform (Modify) Bone - Bone: attach_cape"!(ComponentPose=@n1, bAlphaBoolEnabled:bool=Get HoodDownMode(), Translation="0,-5.5,0", Rotation="0,0,0", Scale="1,1,1") -> (Pose) {
    alphaInputType=Bool
    rotationSpace=BCS_BoneSpace
    translationMode=BMM_Additive
    translationSpace=BCS_BoneSpace
    alphaBoolBlend {blendOption=Linear}
    boneToModify {boneName=attach_cape}
    translation {y=-5.5}
   }
   n3 = "Transform (Modify) Bone - Bone: attach_cape"!(
    ComponentPose=@n2
    Alpha:float=Get hoodDownRotationAlpha() -> (hoodDownRotationAlpha:float)
    Translation="0,0,0"
    Rotation="0,0,15"
    Scale="1,1,1"
   ) -> (Pose) {
    rotationMode=BMM_Additive
    rotationSpace=BCS_BoneSpace
    translationSpace=BCS_BoneSpace
    alphaBoolBlend {blendOption=Linear}
    alphaScaleBiasClamp {bInterpResult=true, interpSpeedDecreasing=7}
    boneToModify {boneName=attach_cape}
    rotation {roll=15}
   }
   Output Pose(Result=Component To Local(ComponentPose=@n3)) {layerGroup=None}
   n4 = Control Rig(Source=Copy Pose From Mesh() {bUseAttachedParent=true}) {
    alphaBoolBlend {blendOption=Linear}
    controlRigAssetReference {blueprintRigClass=&/F_MED_DeformRig_VitalPsych_ControlRig.F_MED_DeformRig_VitalPsych_ControlRig_C}
   }
   n5 = "Transform (Modify) Bone - Bone: deform_glute_l"(
    ComponentPose=Local To Component(LocalPose=@n4)
    Alpha=Get Is Crouching() -> (Is Crouching:double)
    Translation="0,-5,-10"
    Rotation="0,0,0"
    Scale="1,1,1"
   ) {
    translationMode=BMM_Additive
    alphaBoolBlend {blendOption=Linear}
    alphaScaleBiasClamp {bInterpResult=true}
    boneToModify {boneName=deform_glute_l}
    translation {y=-5, z=-10}
   }
   n6 = "Transform (Modify) Bone - Bone: attach_backpack"!(ComponentPose=@n5, Alpha:float=Get BackpackCorrection() -> (BackpackCorrection:double), Rotation="0,0,18") -> (Pose) {
    rotationMode=BMM_Additive
    alphaBoolBlend {blendOption=Linear}
    alphaScaleBiasClamp {bInterpResult=true, interpSpeedDecreasing=4, interpSpeedIncreasing=15}
    boneToModify {boneName=attach_backpack}
    rotation {roll=18}
   }
   n7 = "Constraint - base_front_chest_strap_00_l (PPPPPPPPP)"!(
    ComponentPose=@n6
    Alpha=1
    ConstraintWeights_0=0.4458
    ConstraintWeights_1:float=0.288
    ConstraintWeights_2:float=0.2363
    ConstraintWeights_3:float=0.0275
    ConstraintWeights_4:float=0.0025
    ConstraintWeights_5:float=0
    ConstraintWeights_6:float=0
    ConstraintWeights_7:float=0
    ConstraintWeights_8:float=0
   ) -> (Pose) {
    alphaBoolBlend {blendOption=Linear}
    boneToModify {boneName=base_front_chest_strap_00_l}
    constraintSetup [
     {targetBone {boneName=spine_04}}
     {targetBone {boneName=spine_05}}
     {targetBone {boneName=clavicle_l}}
     {targetBone {boneName=upperarm_twist_01_l}}
     {targetBone {boneName=spine_03}}
     {targetBone {boneName=clavicle_r}}
     {targetBone {boneName=pelvis}}
     {targetBone {boneName=upperarm_twist_01_r}}
     {targetBone {boneName=neck_02}}
    ]
    constraintWeights [1,1,1,1,1,1,1,1,1]
   }
   n8 = "Constraint - base_back_chest_strap_00_l (PPPPPPPP)"!(
    ComponentPose=@n7
    Alpha=1
    ConstraintWeights_0=0.5131
    ConstraintWeights_1:float=0.1972
    ConstraintWeights_2:float=0.1625
    ConstraintWeights_3:float=0.1226
    ConstraintWeights_4:float=0.0046
    ConstraintWeights_5:float=0
    ConstraintWeights_6:float=0
    ConstraintWeights_7:float=0
   ) -> (Pose) {
    alphaBoolBlend {blendOption=Linear}
    boneToModify {boneName=base_back_chest_strap_00_l}
    constraintSetup [
     {targetBone {boneName=clavicle_l}}
     {targetBone {boneName=spine_05}}
     {targetBone {boneName=upperarm_twist_01_l}}
     {targetBone {boneName=spine_04}}
     {targetBone {boneName=spine_03}}
     {targetBone {boneName=pelvis}}
     {targetBone {boneName=neck_02}}
     {targetBone {boneName=upperarm_twist_01_r}}
    ]
    constraintWeights [1,1,1,1,1,1,1,1]
   }
   n9 = "Constraint - base_front_chest_strap_00_r (PPPPPPPPP)"!(
    ComponentPose=@n8
    Alpha=1
    ConstraintWeights_0=0.4458
    ConstraintWeights_1:float=0.288
    ConstraintWeights_2:float=0.2363
    ConstraintWeights_3:float=0.0275
    ConstraintWeights_4:float=0.0025
    ConstraintWeights_5:float=0
    ConstraintWeights_6:float=0
    ConstraintWeights_7:float=0
    ConstraintWeights_8:float=0
   ) -> (Pose) {
    alphaBoolBlend {blendOption=Linear}
    boneToModify {boneName=base_front_chest_strap_00_r}
    constraintSetup [
     {targetBone {boneName=spine_04}}
     {targetBone {boneName=spine_05}}
     {targetBone {boneName=clavicle_r}}
     {targetBone {boneName=upperarm_twist_01_r}}
     {targetBone {boneName=spine_03}}
     {targetBone {boneName=clavicle_l}}
     {targetBone {boneName=pelvis}}
     {targetBone {boneName=upperarm_twist_01_l}}
     {targetBone {boneName=neck_02}}
    ]
    constraintWeights [1,1,1,1,1,1,1,1,1]
   }
   n10 = "Constraint - base_back_chest_strap_00_r (PPPPPPPP)"!(
    ComponentPose=@n9
    Alpha=1
    ConstraintWeights_0=0.5131
    ConstraintWeights_1:float=0.1972
    ConstraintWeights_2:float=0.1625
    ConstraintWeights_3:float=0.1226
    ConstraintWeights_4:float=0.0046
    ConstraintWeights_5:float=0
    ConstraintWeights_6:float=0
    ConstraintWeights_7:float=0
   ) -> (Pose) {
    alphaBoolBlend {blendOption=Linear}
    boneToModify {boneName=base_back_chest_strap_00_r}
    constraintSetup [
     {targetBone {boneName=clavicle_r}}
     {targetBone {boneName=spine_05}}
     {targetBone {boneName=upperarm_twist_01_r}}
     {targetBone {boneName=spine_04}}
     {targetBone {boneName=spine_03}}
     {targetBone {boneName=pelvis}}
     {targetBone {boneName=neck_02}}
     {targetBone {boneName=upperarm_twist_01_l}}
    ]
    constraintWeights [1,1,1,1,1,1,1,1]
   }
   n11 = "Constraint - base_pocket (PPPPPPPPPPPPPP)"!(
    ComponentPose=@n10
    Alpha=1
    ConstraintWeights_0=0.9764
    ConstraintWeights_1:float=0.0211
    ConstraintWeights_2:float=0.0022
    ConstraintWeights_3:float=0.0001
    ConstraintWeights_4:float=0.0001
    ConstraintWeights_5:float=0
    ConstraintWeights_6:float=0
    ConstraintWeights_7:float=0
    ConstraintWeights_8:float=0
    ConstraintWeights_9:float=0
    ConstraintWeights_10:float=0
    ConstraintWeights_11:float=0
    ConstraintWeights_12:float=0
    ConstraintWeights_13:float=0
   ) -> (Pose) {
    alphaBoolBlend {blendOption=Linear}
    boneToModify {boneName=base_pocket}
    constraintSetup [
     {targetBone {boneName=pelvis}}
     {targetBone {boneName=thigh_twist_01_r}}
     {targetBone {boneName=spine_02}}
     {targetBone {boneName=spine_01}}
     {targetBone {boneName=thigh_r}}
     {targetBone {boneName=thigh_twist_01_l}}
     {targetBone {boneName=clavicle_l}}
     {targetBone {boneName=upperarm_twist_01_r}}
     {targetBone {boneName=upperarm_twist_01_l}}
     {targetBone {boneName=neck_02}}
     {targetBone {boneName=upperarm_twist_02_r}}
     {targetBone {boneName=upperarm_r}}
     {targetBone {boneName=upperarm_twist_02_l}}
     {targetBone {boneName=upperarm_l}}
    ]
    constraintWeights [0.9764,0.0211,0.0022,0.0001,0.0001,0,0,0,0,0,0,0,0,0]
   }
   Save cached pose 'Constraints Cache'(Pose=Component To Local(ComponentPose=@n11)) {cacheName=Constraints Cache}
   n12 = Get GravityOverride()
   n13 = RigidBody!(
    ComponentPose=Local To Component(LocalPose=Use cached pose 'NeckCollarCache'())
    Alpha=Get Is Skydive Floating() -> (Is Skydive Floating:double)
    OverrideWorldGravity:Vector=@n12
    ExternalForce
   ) -> (Pose) {
    bClampLinearTranslationLimitToRefPose=true
    bOverrideWorldGravity=true
    overridePhysicsAsset=&/F_MED_VitalPsych_Chest_Physics
    simulationSpace=BaseBoneSpace
    alphaBoolBlend {blendOption=Linear}
    alphaScaleBiasClamp {bClampResult=true, bInterpResult=true, clampMin=0.75}
    baseBoneRef {boneName=pelvis}
    componentAppliedLinearAccClamp=<500 500 500>
    componentLinearAccScale=<0.5 0.3 0.3>
    componentLinearVelScale=<0.3 0.3 0.3>
    overrideWorldGravity {z=-980}
    simSpaceSettings {
     maxAngularAcceleration=100
     maxAngularVelocity=0.5
     maxLinearAcceleration=200
     maxLinearVelocity=200
     velocityScaleZ=0
     worldAlpha=0.5
     externalLinearDragV=<0.2 0.2>
    }
   }
   n14 = RigidBody(ComponentPose=@n13, Alpha=Get RB_LegStrap_Alpha() -> (RB_LegStrap_Alpha:double), OverrideWorldGravity=@n12) {
    bClampLinearTranslationLimitToRefPose=true
    bOverrideWorldGravity=true
    overridePhysicsAsset=&/F_MED_VitalPsych_Physics
    simulationSpace=BaseBoneSpace
    alphaBoolBlend {blendOption=Linear}
    alphaScaleBiasClamp {bInterpResult=true, clampMin=0.75}
    baseBoneRef {boneName=pelvis}
    componentAppliedLinearAccClamp=<500 500 500>
    componentLinearAccScale=<0.5 0.3 0.3>
    componentLinearVelScale=<0.3 0.3 0.3>
    overrideWorldGravity {z=-980}
    simSpaceSettings {
     maxAngularAcceleration=100
     maxAngularVelocity=0.5
     maxLinearAcceleration=200
     maxLinearVelocity=200
     velocityScaleZ=0
     worldAlpha=0.25
     externalLinearDragV=<0.2 0.2>
    }
   }
   n15 = RigidBody(ComponentPose=@n14, OverrideWorldGravity=SelectVector(A="0,0,-750", B=@n12, bPickA=Get bIsSkyDiving())) {
    bClampLinearTranslationLimitToRefPose=true
    bOverrideWorldGravity=true
    overridePhysicsAsset=&/F_MED_VitalPsych_BackChest_Physics
    simulationSpace=BaseBoneSpace
    alphaBoolBlend {blendOption=Linear}
    alphaScaleBiasClamp {clampMin=0.75}
    baseBoneRef {boneName=spine_02}
    componentAppliedLinearAccClamp=<1000 1000 1000>
    componentLinearAccScale=<0.5 0.3 0.3>
    componentLinearVelScale=<0.3 0.3 0.3>
    overrideWorldGravity {z=-980}
    simSpaceSettings {
     maxAngularAcceleration=100
     maxAngularVelocity=0.5
     maxLinearAcceleration=200
     maxLinearVelocity=200
     velocityScaleZ=0
     externalLinearDragV=<0.2 0.2>
    }
   }
   n16 = Get GravityOverride()
   n17 = RigidBody(ComponentPose=Local To Component(LocalPose=Use cached pose 'NeckCollarCache'()), Alpha=Get Is Skydive Floating(), OverrideWorldGravity=@n16) {
    bClampLinearTranslationLimitToRefPose=true
    bOverrideWorldGravity=true
    overridePhysicsAsset=&F_MED_VitalPsych_Chest_Physics
    simulationSpace=BaseBoneSpace
    alphaBoolBlend {blendOption=Linear}
    alphaScaleBiasClamp {bClampResult=true, bInterpResult=true, clampMin=0.75}
    baseBoneRef {boneName=pelvis}
    componentAppliedLinearAccClamp=<500 500 500>
    componentLinearAccScale=<0.5 0.3 0.3>
    componentLinearVelScale=<0.3 0.3 0.3>
    overrideWorldGravity {z=-980}
    simSpaceSettings {
     maxAngularAcceleration=100
     maxAngularVelocity=0.5
     maxLinearAcceleration=200
     maxLinearVelocity=200
     velocityScaleZ=0
     worldAlpha=0.5
     externalLinearDragV=<0.2 0.2>
    }
   }
   n18 = RigidBody(ComponentPose=@n17, Alpha=Get Is Skydive Floating(), OverrideWorldGravity=@n16) {
    bClampLinearTranslationLimitToRefPose=true
    bOverrideWorldGravity=true
    overridePhysicsAsset=&F_MED_VitalPsych_Physics
    simulationSpace=BaseBoneSpace
    alphaBoolBlend {blendOption=Linear}
    alphaScaleBiasClamp {bClampResult=true, bInterpResult=true, clampMin=0.75}
    baseBoneRef {boneName=pelvis}
    componentAppliedLinearAccClamp=<500 500 500>
    componentLinearAccScale=<0.5 0.3 0.3>
    componentLinearVelScale=<0.3 0.3 0.3>
    overrideWorldGravity {z=-980}
    simSpaceSettings {
     maxAngularAcceleration=100
     maxAngularVelocity=0.5
     maxLinearAcceleration=200
     maxLinearVelocity=200
     velocityScaleZ=0
     worldAlpha=0.25
     externalLinearDragV=<0.2 0.2>
    }
   }
   n19 = Look At - Bone: base_back_chest_strap_00_l(ComponentPose=@n18, Alpha=0.25) {
    alphaBoolBlend {blendOption=Linear}
    boneToModify {boneName=base_back_chest_strap_00_l}
    lookAtLocation=<0.8595 -1.1464 -13.5132>
    lookAtTarget {boneReference {boneName=pelvis}}
    lookAt_Axis {axis=<1 0>}
    lookUp_Axis {axis {x=0, z=1}}
   }
   n20 = Look At - Bone: base_back_chest_strap_00_r(ComponentPose=@n19, Alpha=0.25) {
    alphaBoolBlend {blendOption=Linear}
    boneToModify {boneName=base_back_chest_strap_00_r}
    lookAtLocation=<0.8595 -1.1464 -13.5132>
    lookAtTarget {boneReference {boneName=pelvis}}
    lookAt_Axis {axis=<1 0>}
    lookUp_Axis {axis {x=0, z=1}}
   }
   n21 = Control Rig(Source=Blend Poses by bool(BlendPose_0=Component To Local(ComponentPose=@n20), BlendPose_1=Component To Local(ComponentPose=@n15), bActiveValue=Get HoodDownMode()) {blendType=Linear}) {
    alphaBoolBlend {blendOption=Linear}
    controlRigAssetReference {blueprintRigClass=&/F_MED_VitalPsych_CtrlRig.F_MED_VitalPsych_CtrlRig_C}
   }
   n22 = "Transform (Modify) Bone - Bone: dyn_back_leg_strap_00"(
    ComponentPose=Local To Component(LocalPose=@n21)
    Alpha=Get Is Jumping() -> (Is Jumping:double)
    Translation="-4,0,0"
    Rotation="0,0,0"
    Scale="1,1,1"
   ) {
    translationMode=BMM_Additive
    translationSpace=BCS_BoneSpace
    alphaBoolBlend {blendOption=Linear}
    alphaScaleBiasClamp {bInterpResult=true}
    boneToModify {boneName=dyn_back_leg_strap_00}
    translation {x=-4}
   }
   n23 = Trail controller - Bone: dyn_front_chest_strap_01_l!(
    ComponentPose=Local To Component(LocalPose=Use cached pose 'NeckCollarCache'())
    Alpha=0.25
    RelaxationSpeedScale:float=2
    FakeVelocity:Vector="0,0,0"
   ) -> (Pose) {
    bActorSpaceFakeVel=true
    bShowTrailLocation=true
    bUsePlanarLimit=true
    debugLifeTime=0.1
    maxDeltaTime=0.0333
    relaxationSpeedScale=1.2
    alphaScaleBiasClamp {interpSpeedDecreasing=25, interpSpeedIncreasing=25}
    baseJoint {boneName=root}
    rotationLimits [{limitMax=<180 180 180>}, {limitMax=<180 180 180>}]
    rotationOffsets [<0 0 0>, <0 0 0>]
    trailBone {boneName=dyn_front_chest_strap_01_l}
    trailRelaxationSpeed {editorCurveData {keys [<at 0 15 RCIM_Linear>,<at 1 6 RCIM_Linear>]}}
   }
   n24 = Trail controller - Bone: dyn_front_chest_strap_01_r!(ComponentPose=@n23, Alpha=0.25, RelaxationSpeedScale:float=2, FakeVelocity:Vector="0,0,0") -> (Pose) {
    bActorSpaceFakeVel=true
    bShowTrailLocation=true
    bUsePlanarLimit=true
    debugLifeTime=0.1
    maxDeltaTime=0.0333
    relaxationSpeedScale=1.2
    alphaScaleBiasClamp {interpSpeedDecreasing=25, interpSpeedIncreasing=25}
    baseJoint {boneName=root}
    rotationLimits [{limitMax=<180 180 180>}, {limitMax=<180 180 180>}]
    rotationOffsets [<0 0 0>, <0 0 0>]
    trailBone {boneName=dyn_front_chest_strap_01_r}
    trailRelaxationSpeed {editorCurveData {keys [<at 0 15 RCIM_Linear>,<at 1 6 RCIM_Linear>]}}
   }
   n25 = SelectFloat(A=0.2, B=0, bPickA=Get HoodUpMode() -> (HoodUpMode:bool))
   n26 = Trail controller - Bone: dyn_back_chest_strap_01_l!(ComponentPose=@n24, Alpha=@n25, RelaxationSpeedScale:float=2, FakeVelocity:Vector="0,0,0") -> (Pose) {
    bActorSpaceFakeVel=true
    bShowTrailLocation=true
    bUsePlanarLimit=true
    debugLifeTime=0.1
    maxDeltaTime=0.0333
    relaxationSpeedScale=1.2
    alphaScaleBiasClamp {interpSpeedDecreasing=25, interpSpeedIncreasing=25}
    baseJoint {boneName=root}
    rotationLimits [{limitMax=<180 180 180>}, {limitMax=<180 180 180>}]
    rotationOffsets [<0 0 0>, <0 0 0>]
    trailBone {boneName=dyn_back_chest_strap_01_l}
    trailRelaxationSpeed {editorCurveData {keys [<at 0 15 RCIM_Linear>,<at 1 6 RCIM_Linear>]}}
   }
   n27 = Trail controller - Bone: dyn_back_chest_strap_01_r!(ComponentPose=@n26, Alpha=@n25, RelaxationSpeedScale:float=2, FakeVelocity:Vector="0,0,0") -> (Pose) {
    bActorSpaceFakeVel=true
    bShowTrailLocation=true
    bUsePlanarLimit=true
    debugLifeTime=0.1
    maxDeltaTime=0.0333
    relaxationSpeedScale=1.2
    alphaScaleBiasClamp {interpSpeedDecreasing=25, interpSpeedIncreasing=25}
    baseJoint {boneName=root}
    rotationLimits [{limitMax=<180 180 180>}, {limitMax=<180 180 180>}]
    rotationOffsets [<0 0 0>, <0 0 0>]
    trailBone {boneName=dyn_back_chest_strap_01_r}
    trailRelaxationSpeed {editorCurveData {keys [<at 0 15 RCIM_Linear>,<at 1 6 RCIM_Linear>]}}
   }
   Save cached pose 'StrapsCache'(Pose=Blend Poses by bool(BlendPose_0=Component To Local(ComponentPose=@n22), BlendPose_1=Component To Local(ComponentPose=@n27), bActiveValue=Get bRigidBodyEnabled()) {blendType=Linear}) {cacheName=StrapsCache}
   n28 = Get HoodDownMode()
   n29 = "Transform (Modify) Bone - Bone: base_neck_collar_c"!(
    ComponentPose=Local To Component(LocalPose=Use cached pose 'Constraints Cache'())
    bAlphaBoolEnabled:bool=@n28
    Translation="0,0,0"
    Rotation="0,0,0"
    Scale="0.2,0.2,0.2"
   ) -> (Pose) {
    alphaInputType=Bool
    scaleMode=BMM_Additive
    scaleSpace=BCS_BoneSpace
    alphaBoolBlend {blendOption=Linear}
    boneToModify {boneName=base_neck_collar_c}
    scale=<0.2 0.2 0.2>
   }
   n30 = "Transform (Modify) Bone - Bone: base_neck_collar_00_r"!(ComponentPose=@n29, bAlphaBoolEnabled:bool=@n28, Translation="0,0,0", Rotation="0,0,0", Scale="0.2,0.2,0.2") -> (Pose) {
    alphaInputType=Bool
    scaleMode=BMM_Additive
    scaleSpace=BCS_BoneSpace
    alphaBoolBlend {blendOption=Linear}
    boneToModify {boneName=base_neck_collar_00_r}
    scale=<0.2 0.2 0.2>
   }
   n31 = "Transform (Modify) Bone - Bone: base_neck_collar_00_l"!(ComponentPose=@n30, bAlphaBoolEnabled:bool=@n28, Translation="0,0,0", Rotation="0,0,0", Scale="0.2,0.2,0.2") -> (Pose) {
    alphaInputType=Bool
    scaleMode=BMM_Additive
    scaleSpace=BCS_BoneSpace
    alphaBoolBlend {blendOption=Linear}
    boneToModify {boneName=base_neck_collar_00_l}
    scale=<0.2 0.2 0.2>
   }
   n32 = Control Rig!(
    Source=Use cached pose 'Constraints Cache'()
    Alpha
    Crouch_bool:bool=Get bIsCrouching()
    Aiming_alpha:double=Get isTargeting_alpha() -> (isTargeting_alpha:double)
    Crouch_alpha:double=Get Is Crouching()
    Jump_alpha:double=Get Is Jumping_interp() -> (Is Jumping_interp:double)
   ) -> (Pose) {
    alphaBoolBlend {blendOption=Linear}
    controlRigAssetReference {blueprintRigClass=&/F_MED_VitalPsych_NeckCollar_CtrlRig.F_MED_VitalPsych_NeckCollar_CtrlRig_C}
   }
   n33 = Get HeadRotate_Front() -> (HeadRotate_Front:double)
   n34 = "Transform (Modify) Bone - Bone: neck_collar_front_c"!(ComponentPose=Local To Component(LocalPose=@n32), Alpha=@n33) -> (Pose) {
    rotationMode=BMM_Additive
    rotationSpace=BCS_BoneSpace
    translationMode=BMM_Additive
    translationSpace=BCS_BoneSpace
    alphaBoolBlend {blendOption=Linear}
    boneToModify {boneName=neck_collar_front_c}
    rotation {yaw=-10}
    translation=<-2 -4.432>
   }
   n35 = "Transform (Modify) Bone - Bone: neck_collar_02_l"!(ComponentPose=@n34, Alpha=@n33) -> (Pose) {
    rotationMode=BMM_Additive
    rotationSpace=BCS_BoneSpace
    translationMode=BMM_Additive
    translationSpace=BCS_BoneSpace
    alphaBoolBlend {blendOption=Linear}
    boneToModify {boneName=neck_collar_02_l}
    rotation {yaw=-40}
    translation {y=-2.133, z=2.133}
   }
   n36 = "Transform (Modify) Bone - Bone: neck_collar_02_r"!(ComponentPose=@n35, Alpha=@n33) -> (Pose) {
    rotationMode=BMM_Additive
    rotationSpace=BCS_BoneSpace
    translationMode=BMM_Additive
    translationSpace=BCS_BoneSpace
    alphaBoolBlend {blendOption=Linear}
    boneToModify {boneName=neck_collar_02_r}
    rotation {yaw=-40}
    translation {y=-2.133, z=-2.133}
   }
   n37 = "Transform (Modify) Bone - Bone: neck_collar_01_l"!(ComponentPose=@n36, Alpha=@n33) -> (Pose) {
    rotationMode=BMM_Additive
    rotationSpace=BCS_BoneSpace
    translationMode=BMM_Additive
    translationSpace=BCS_BoneSpace
    alphaBoolBlend {blendOption=Linear}
    boneToModify {boneName=neck_collar_01_l}
    rotation {roll=20, yaw=-10}
    translation {y=-2}
   }
   n38 = "Transform (Modify) Bone - Bone: neck_collar_01_r"!(ComponentPose=@n37, Alpha=@n33) -> (Pose) {
    rotationMode=BMM_Additive
    rotationSpace=BCS_BoneSpace
    translationMode=BMM_Additive
    translationSpace=BCS_BoneSpace
    alphaBoolBlend {blendOption=Linear}
    boneToModify {boneName=neck_collar_01_r}
    rotation {roll=-20, yaw=-10}
    translation {y=-2}
   }
   n39 = Get HeadRotate_L() -> (HeadRotate_L:double)
   n40 = "Transform (Modify) Bone - Bone: neck_collar_00_l"!(ComponentPose=@n38, Alpha=@n39) -> (Pose) {
    rotationMode=BMM_Additive
    rotationSpace=BCS_BoneSpace
    translationMode=BMM_Additive
    translationSpace=BCS_BoneSpace
    alphaBoolBlend {blendOption=Linear}
    boneToModify {boneName=neck_collar_00_l}
    rotation {roll=11.078}
    translation=<0.679 -2.901 -1.73>
   }
   n41 = "Transform (Modify) Bone - Bone: neck_collar_01_l"(ComponentPose=@n40, Alpha=@n39) {
    rotationMode=BMM_Additive
    rotationSpace=BCS_BoneSpace
    translationMode=BMM_Additive
    translationSpace=BCS_BoneSpace
    alphaBoolBlend {blendOption=Linear}
    boneToModify {boneName=neck_collar_01_l}
    rotation=<rot 2.172 -3.085 -0.811>
    translation=<-0.849 -3.726 0.492>
   }
   n42 = "Transform (Modify) Bone - Bone: neck_collar_02_l"(ComponentPose=@n41, Alpha=@n39) {
    rotationMode=BMM_Additive
    rotationSpace=BCS_BoneSpace
    translationMode=BMM_Additive
    translationSpace=BCS_BoneSpace
    alphaBoolBlend {blendOption=Linear}
    boneToModify {boneName=neck_collar_02_l}
    rotation {yaw=-5.555}
    translation=<-0.668 -3.231 3.321>
   }
   n43 = "Transform (Modify) Bone - Bone: neck_collar_front_c"(ComponentPose=@n42, Alpha=@n39) {
    rotationMode=BMM_Additive
    rotationSpace=BCS_BoneSpace
    translationMode=BMM_Additive
    translationSpace=BCS_BoneSpace
    alphaBoolBlend {blendOption=Linear}
    boneToModify {boneName=neck_collar_front_c}
    rotation {yaw=-4.189}
    translation {y=-1.834}
   }
   n44 = "Transform (Modify) Bone - Bone: neck_collar_02_l"(ComponentPose=@n43, Alpha=Get HeadTwist_R() -> (HeadTwist_R:double)) {
    rotationMode=BMM_Additive
    rotationSpace=BCS_BoneSpace
    alphaBoolBlend {blendOption=Linear}
    boneToModify {boneName=neck_collar_02_l}
    rotation {yaw=-40}
   }
   Save cached pose 'NeckCollarCache'(Pose=Blend Poses by bool(BlendPose_0=Component To Local(ComponentPose=@n44), BlendPose_1=Component To Local(ComponentPose=@n31), bActiveValue=Get HoodUpMode()) {blendType=Linear}) {cacheName=NeckCollarCache}
   Get Is Skydive Floating()
  ]}
 EventGraph {nodes [
   n1 = Event BlueprintUpdateAnimation()
   > n2 = Sequence!(execute) -> (then_0, then_1, then_2:Exec, then_3:Exec)
   > n3 = Gravity Override State Function DEPRECATED(
    CharacterPartAnimInstance=Self-Reference()
    GameTimeInSeconds=GetGameTimeInSeconds()
    GravityOverrideParams=Get GravityOverrideParams() -> (GravityOverrideParams:Gravity Override Params Struct Structure)
   )
   > Set GravityOverride(GravityOverride=@n3)
   > n4 = Cast To FortPlayerPawn(Object=GetOwningActor())
   > n5 = GetVariantChannelActiveIndexForItemActor!(
    execute
    Item:Athena Cosmetic Item Definition Object Reference=&Character_VitalPsych
    Actor:Actor Object Reference=@n4
    Channel:Gameplay Tag Structure=(TagName="Cosmetics.Variant.Channel.Progressive")
    VariantList:Array of Gameplay Tag Structures=Make Array!("[0]"=(TagName="Cosmetics.Variant.Property.Stage1"), "[1]":Gameplay Tag Structure=(TagName="Cosmetics.Variant.Property.Stage3")) -> (Array)
    bErrorIfChannelMissing:bool=false
    bErrorIfTagNotInArray:bool=false
   ) -> (then, Index:int=0, bIndexIsValid:bool, bFoundChannel:bool)
   n6 = "Equal (Integer)"(A=@n5)
   n7 = "Equal (Integer)"(A=@n5, B=1)
   > n8 = Branch(Condition=AND Boolean(A=OR Boolean(A=@n6, B=@n7), B=@n5.bIndexIsValid))
   > Set HoodUpMode(HoodUpMode:bool=true) -> (Output_Get:bool)
   > Set HoodDownMode(HoodDownMode:bool=false) -> (Output_Get:bool)
   Set HoodUpMode(execute=@n8.else, HoodUpMode=false)
   > Set HoodDownMode(HoodDownMode=true)
   n9 = "To Float (Boolean)"(InBool=OR Boolean(A=Get bIsJumping(), B=Get bIsFalling()))
   Set Is Jumping(execute=@n2.then_1, Is Jumping:double=@n9) -> (Output_Get:double)
   > Set Is Jumping_interp(Is Jumping_interp:double=FInterpTo(Current=Get Is Jumping_interp(), Target=@n9, DeltaTime=@n1.DeltaTimeX, InterpSpeed=7)) -> (Output_Get:double)
   n10 = NOT Boolean(A=Get bIsCrouching())
   > Set Is Skydive Floating(Is Skydive Floating:double=SelectFloat(A=0, B="To Float (Boolean)"(InBool=@n10), bPickA=Get bIsSkydiveFloating())) -> (Output_Get:double)
   n11 = Get bIsJumping()
   n12 = "To Float (Boolean)"(InBool=OR Boolean!(A=Get bIsSkydiveDiveMode(), B=Get bIsSkydiveFloating(), C:bool=@n11) -> (ReturnValue))
   n13 = Get bIsSprinting()
   > Set BackpackCorrection(BackpackCorrection:double=SelectFloat(A=1, B=SelectFloat(A=0.15, B=@n12, bPickA=@n13), bPickA=AND Boolean(A=@n13, B=@n11))) -> (Output_Get:double)
   n14 = Get bIsTargeting()
   > Set isTargeting_alpha(isTargeting_alpha:double=FInterpTo(Current=Get isTargeting_alpha(), Target="To Float (Boolean)"(InBool=@n14), DeltaTime=@n1.DeltaTimeX, InterpSpeed=15)) -> (Output_Get:double)
   n15 = Get bIsCrouching()
   > Set Is Crouching(Is Crouching:double=FInterpTo(Current=Get Is Crouching(), Target="To Float (Boolean)"(InBool=@n15), DeltaTime=@n1.DeltaTimeX, InterpSpeed=20)) -> (Output_Get:double)
   > n16 = Branch(Condition=Get HoodDownMode())
   > n17 = Set Alpha Hood Down(Alpha Hood Down:float=1) -> (Output_Get:float)
   n18 = Branch(execute=[@n19.then, @n17.then], Condition=OR Boolean!(A=Get bIsJumping(), B=Get bIsMoving2D(), C=Get bIsFalling(), D:bool) -> (ReturnValue))
   > Set hoodDownRotationAlpha(hoodDownRotationAlpha:float=Get Positive() -> (Positive:float)) -> (Output_Get:float)
   Set hoodDownRotationAlpha(execute=@n18.else, hoodDownRotationAlpha=Get Negative() -> (Negative:float))
   n19 = Set Alpha Hood Down(execute=@n16.else, Alpha Hood Down=0)
   Set FrontEnd_Emote_bool(execute=@n2.then_2, FrontEnd_Emote_bool:bool=OR Boolean!(A=Get bIsInFrontEnd(), B=Get bIsPlayingEmote()) -> (ReturnValue)) -> (Output_Get:bool)
   n20 = Get Skel Mesh()
   > Set HeadRotate_Front(HeadRotate_Front:double=float * float(
     A=GetAlphaFromReferenceTransformDelta(SkeletalMeshComponent=@n20, BoneName=head, BaseName=spine_05, InputRange=(X=0,Y=25), TwistAxis="0,0,1")
     B=0.4
    ) {is=title K2Node_PromotableOperator}) -> (Output_Get:double)
   n21 = GetAlphaFromReferenceTransformDelta(SkeletalMeshComponent=Get Skel Mesh(), BoneName=neck_01, BaseName=head, DeltaComponentType=EulerY, InputRange=(X=0,Y=20), TwistAxis="0,0,1")
   > Set HeadRotate_L(HeadRotate_L:double=float * float(A=@n21, B=0.65)) -> (Output_Get:double)
   n22 = GetAlphaFromReferenceTransformDelta(SkeletalMeshComponent=Get Skel Mesh(), BoneName=neck_02, BaseName=head, DeltaComponentType=EulerX, InputRange=(X=0,Y=90), TwistAxis="0,0,1")
   > Set HeadTwist_R(HeadTwist_R:double=float * float(A=@n21, B=float * float(A=@n22, B=1.5))) -> (Output_Get:double)
   n23 = Branch(execute=@n2.then_3, Condition=OR Boolean(A=Get bIsSkydiveFloating(), B=Get bIsCrouching()))
   > Set RB_LegStrap_Alpha(RB_LegStrap_Alpha:double=0.75) -> (Output_Get:double)
   n24 = Branch(execute=@n23.else, Condition=Get bIsSprinting())
   > n25 = Branch(Condition=Get bIsTacticalSprinting())
   > Set RB_LegStrap_Alpha(RB_LegStrap_Alpha=0.25)
   Set RB_LegStrap_Alpha(execute=@n25.else, RB_LegStrap_Alpha=0.5)
   Set RB_LegStrap_Alpha(execute=@n24.else, RB_LegStrap_Alpha=1)
   Event BlueprintInitializeAnimation()
   > Set Skel Mesh(Skel Mesh=GetOwnerSkeletalMeshComponent())
  ]}
]
variables [
 Alpha Hood Down {type=number}
 BackpackCorrection {type=number}
 FrontEnd_Emote_bool {type=boolean}
 GravityOverride {type=Vector}
 GravityOverrideParams {type=GravityOverrideParamsStruct}
 HeadRotate_Front {type=number}
 HeadRotate_L {type=number}
 HeadRotate_R {type=number}
 HeadTwist_L {type=number}
 HeadTwist_R {type=number}
 HoodDownMode {type=boolean}
 hoodDownRotationAlpha {type=number, default=1}
 HoodUpMode {type=boolean}
 Is Crouching {type=number}
 Is Jumping {type=number}
 Is Jumping_interp {type=number}
 Is Skydive Floating {type=number}
 isTargeting_alpha {type=number}
 Negative {type=number}
 Positive {type=number, default=1}
 RB_LegStrap_Alpha {type=number}
 Skel Mesh {type=SkeletalMeshComponent}
]
