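# Physics asset for the ears face accessory of the VitalPsych female medium cosmetic body, giving its hood and ear bones capsule bodies and joints.
# Constraints chain head to base and dynamic hood bones, which branch into left and right ear bases, all with free angular motion.
# Each ear continues as a dynamic bone chain, six left and five right, with tight swing limits, mostly narrower toward the tip, and minimal twist.
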PhysicsAsset /BRCosmetics/Characters/Player/Female/Medium/Bodies/F_MED_VitalPsych/Meshes/Parts/F_MED_VitalPsych_FaceAcc_Ears_Physics
bodies {
 base_ear_00_l [{shapeType=Capsule, length=2, radius=2, shapeName=base_ear_00_l_capsule, rotation=<rot 90 0 0>}]
 base_ear_00_r [{shapeType=Capsule, length=2, radius=2, shapeName=base_ear_00_r_capsule, rotation=<rot 90 0 0>}]
 base_hood_up [{shapeType=Capsule, length=0.505, radius=0.505, shapeName=base_hood_up_capsule, rotation=<rot 0 -90 0>}]
 dyn_ear_00_l [{shapeType=Capsule, length=2, radius=2, shapeName=dyn_ear_00_l_capsule, center=<2.6781 0 0>, rotation=<rot 90 0 0>}]
 dyn_ear_00_r [{shapeType=Capsule, length=2, radius=2, shapeName=dyn_ear_00_r_capsule, center=<3.4603 0 0>, rotation=<rot 90 0 0>}]
 dyn_ear_01_l [{shapeType=Capsule, length=2, radius=2, shapeName=dyn_ear_01_l_capsule, center=<2.5177 0 0>, rotation=<rot 90 0 0>}]
 dyn_ear_01_r [{shapeType=Capsule, length=2, radius=2, shapeName=dyn_ear_01_r_capsule, center=<2.1624 0 0>, rotation=<rot 90 0 0>}]
 dyn_ear_02_l [{shapeType=Capsule, length=2, radius=2, shapeName=dyn_ear_02_l_capsule, center=<2.5892 0 0>, rotation=<rot 90 0 0>}]
 dyn_ear_02_r [{shapeType=Capsule, length=2, radius=2, shapeName=dyn_ear_02_r_capsule, center=<2.9052 0 0>, rotation=<rot 90 0 0>}]
 dyn_ear_03_l [{shapeType=Capsule, length=2, radius=2, shapeName=dyn_ear_03_l_capsule, center=<2.1802 0 0>, rotation=<rot 90 0 0>}]
 dyn_ear_03_r [{shapeType=Capsule, length=2, radius=2, shapeName=dyn_ear_03_r_capsule, center=<3.1845 0 0>, rotation=<rot 90 0 0>}]
 dyn_ear_04_l [{shapeType=Capsule, length=2, radius=2, shapeName=dyn_ear_04_l_capsule, center=<2.6855 0 0>, rotation=<rot 90 0 0>}]
 dyn_ear_04_r [{shapeType=Capsule, length=2, radius=2, shapeName=dyn_ear_04_r_capsule, rotation=<rot 90 0 0>}]
 dyn_ear_05_l [{shapeType=Capsule, length=2, radius=2, shapeName=dyn_ear_05_l_capsule, rotation=<rot 90 0 0>}]
 dyn_hood_up [{shapeType=Capsule, length=1, radius=1}]
 head [{shapeType=Capsule, length=1, radius=1}]
}
constraints [
 {
  bone1Name=base_ear_00_l
  bone2Name=dyn_hood_up
  swing1LimitDegrees=45
  swing1Motion=Free
  swing2LimitDegrees=45
  swing2Motion=Free
  twistLimitDegrees=45
  twistMotion=Free
 }
 {bone1Name=dyn_ear_00_l, bone2Name=base_ear_00_l, swing1LimitDegrees=7, swing2LimitDegrees=7, twistLimitDegrees=1}
 {bone1Name=dyn_ear_01_l, bone2Name=dyn_ear_00_l, swing1LimitDegrees=7, swing2LimitDegrees=7, twistLimitDegrees=1}
 {bone1Name=dyn_ear_02_l, bone2Name=dyn_ear_01_l, swing1LimitDegrees=5, swing2LimitDegrees=5, twistLimitDegrees=1}
 {bone1Name=dyn_ear_03_l, bone2Name=dyn_ear_02_l, swing1LimitDegrees=5, swing2LimitDegrees=5, twistLimitDegrees=1}
 {bone1Name=dyn_ear_04_l, bone2Name=dyn_ear_03_l, swing1LimitDegrees=5, swing2LimitDegrees=5, twistLimitDegrees=1}
 {bone1Name=dyn_ear_05_l, bone2Name=dyn_ear_04_l, swing1LimitDegrees=5, swing2LimitDegrees=5, twistLimitDegrees=1}
 {
  bone1Name=base_ear_00_r
  bone2Name=dyn_hood_up
  swing1LimitDegrees=45
  swing1Motion=Free
  swing2LimitDegrees=45
  swing2Motion=Free
  twistLimitDegrees=45
  twistMotion=Free
 }
 {bone1Name=dyn_ear_00_r, bone2Name=base_ear_00_r, swing1LimitDegrees=7, swing2LimitDegrees=7, twistLimitDegrees=1}
 {bone1Name=dyn_ear_01_r, bone2Name=dyn_ear_00_r, swing1LimitDegrees=7, swing2LimitDegrees=7, twistLimitDegrees=1}
 {bone1Name=dyn_ear_02_r, bone2Name=dyn_ear_01_r, swing1LimitDegrees=7, swing2LimitDegrees=7, twistLimitDegrees=1}
 {bone1Name=dyn_ear_03_r, bone2Name=dyn_ear_02_r, swing1LimitDegrees=5, swing2LimitDegrees=5, twistLimitDegrees=1}
 {bone1Name=dyn_ear_04_r, bone2Name=dyn_ear_03_r, swing1LimitDegrees=5, swing2LimitDegrees=5, twistLimitDegrees=1}
 {
  bone1Name=base_hood_up
  bone2Name=head
  swing1LimitDegrees=45
  swing1Motion=Free
  swing2LimitDegrees=45
  swing2Motion=Free
  twistLimitDegrees=45
  twistMotion=Free
 }
 {
  bone1Name=dyn_hood_up
  bone2Name=base_hood_up
  swing1LimitDegrees=45
  swing1Motion=Free
  swing2LimitDegrees=45
  swing2Motion=Free
  twistLimitDegrees=45
  twistMotion=Free
 }
]
