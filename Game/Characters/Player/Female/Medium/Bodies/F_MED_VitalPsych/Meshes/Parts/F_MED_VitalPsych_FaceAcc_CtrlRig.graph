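# Control Rig for the VitalPsych female player's face accessory mesh, applying head-driven corrective rotations to the ponytail base bones.
# A sequence runs two spherical pose readers on the head bone, one relative to spine_03 and the other to pelvis, with differently skewed regions.
# Each reader's output weights a Modify Transforms node that turns both ponytail bases, in opposite directions between the two branches.
# Debug drawing is left enabled on both readers.
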ControlRigBlueprint /Game/Characters/Player/Female/Medium/Bodies/F_MED_VitalPsych/Meshes/Parts/F_MED_VitalPsych_FaceAcc_CtrlRig
graphs [
 RigVMFunctionLibrary {links [], nodes []}
 RigVMModel {nodes [
   RigUnit_BeginExecution = &RigUnit_BeginExecution()
   SequenceAggregate = &RigVMFunction_Sequence(ExecuteContext=@RigUnit_BeginExecution.ExecutePin)
   n1 = &RigUnit_SphericalPoseReader(
    ActiveRegionScaleFactors=(PositiveWidth=6,NegativeWidth=1,PositiveHeight=4,NegativeHeight=4)
    ActiveRegionSize=0.1
    Debug=(bDrawDebug=True,DebugScale=25,DebugSegments=20,DebugThickness=0.25)
    DriverAxis=(X=1,Y=0,Z=0)
    DriverItem=(Type=Bone,Name="head")
    FalloffRegionScaleFactors=(PositiveWidth=0,NegativeWidth=1,PositiveHeight=0,NegativeHeight=0)
    FalloffSize=0.275
    OptionalParentItem=(Type=Bone,Name="spine_03")
    RotationOffset=(X=90,Y=0,Z=-145)
    ExecutePin=@SequenceAggregate.A
   )
   > &RigUnit_ModifyTransforms(
    ItemToModify="((Item=(Type=Bone,Name=\"base_ponytail_00_l\"),Transform=(Rotation=(X=0,Y=0,Z=-0.2588,W=0.9659),Translation=(X=0,Y=0,Z=0),Scale3D=(X=1,Y=1,Z=1))),(Item=(Type=Bone,Name=\"base_ponytail_00_r\"),Transform=(Rotation=(X=0,Y=0,Z=-0.2588,W=0.9659),Translation=(X=0,Y= ...(truncated)"
    Weight=@n1.OutputParam
   )
   n2 = &RigUnit_SphericalPoseReader(
    ActiveRegionScaleFactors=(PositiveWidth=1,NegativeWidth=6,PositiveHeight=4,NegativeHeight=4)
    ActiveRegionSize=0.1
    Debug=(bDrawDebug=True,DebugScale=25,DebugSegments=20,DebugThickness=0.25)
    DriverAxis=(X=1,Y=0,Z=0)
    DriverItem=(Type=Bone,Name="head")
    FalloffRegionScaleFactors=(PositiveWidth=1,NegativeWidth=0,PositiveHeight=0,NegativeHeight=0)
    FalloffSize=0.275
    OptionalParentItem=(Type=Bone,Name="pelvis")
    RotationOffset=(X=90,Y=0,Z=-35)
    ExecutePin=@SequenceAggregate.B
   )
   > &RigUnit_ModifyTransforms(
    ItemToModify="((Item=(Type=Bone,Name=\"base_ponytail_00_l\"),Transform=(Rotation=(X=0,Y=0,Z=0.1305,W=0.9914),Translation=(X=0,Y=0,Z=0),Scale3D=(X=1,Y=1,Z=1))),(Item=(Type=Bone,Name=\"base_ponytail_00_r\"),Transform=(Rotation=(X=0,Y=0,Z=0.1305,W=0.9914),Translation=(X=0,Y=0, ...(truncated)"
    Weight=@n2.OutputParam
   )
  ]}
]
hierarchy [
 root {loc=<0 0 0>, rot=<0 0 0 1>}
 pelvis {parent=root, loc=<0 -0.5301 98.0571>, rot=<0.0127 0.707 -0.0127 -0.707>}
 spine_01 {parent=pelvis, loc=<6.1208 -0.9328 0>, rot=<0 0 0.034 -0.9994>}
 spine_02 {parent=spine_01, loc=<6.5996 -0.2095 0>, rot=<0 0 0.0238 -0.9997>}
 spine_03 {parent=spine_02, loc=<6.4914 -0.1237 0>, rot=<0 0 0.0149 -0.9999>}
 spine_04 {parent=spine_03, loc=<6.9598 0.0794 0>, rot=<0 0 -0.0132 -0.9999>}
 spine_05 {parent=spine_04, loc=<6.9987 0 0>, rot=<0 0 -0.0015 -1>}
 attach_cape {parent=spine_05, loc=<9.3216 -8.4501 0>, rot=<0.0616 -0.7044 -0.0616 -0.7044>}
 neck_01 {parent=spine_05, loc=<11.3138 0.1529 0>, rot=<0 0 -0.2003 -0.9797>}
 neck_02 {parent=neck_01, loc=<4.5627 0 0>, rot=<0 0 0.0504 -0.9987>}
 head {parent=neck_02, loc=<5.3914 0 0>, rot=<0 0 0.115 -0.9934>}
 base_hair_c {parent=head, loc=<7.0035 8.8143 -0.6814>, rot=<0.158 0.0184 0.9815 0.1063>}
 dyn_hair_c {parent=base_hair_c, loc=<0 0 0>, rot=<0 0 0 1>}
 base_hair_l {parent=head, loc=<2.2275 2.3179 -6.3375>, rot=<0.0872 0.6503 0.7542 -0.0246>}
 dyn_hair_l {parent=base_hair_l, loc=<0 0 0>, rot=<0 0 0 1>}
 base_hair_r {parent=head, loc=<0.8795 2.1523 6.554>, rot=<0.1971 -0.7059 0.68 0.0201>}
 dyn_hair_r {parent=base_hair_r, loc=<0 0 0>, rot=<0 0 0 1>}
 base_ponytail_00_r {parent=head, loc=<-3.4792 -1.5415 4.8294>, rot=<-0.536 -0.8434 -0.033 0.0153>}
 dyn_ponytail_00_r {parent=base_ponytail_00_r, loc=<0 0 0>, rot=<0 0 0 1>}
 dyn_ponytail_01_r {parent=dyn_ponytail_00_r, loc=<4.3129 0 0>, rot=<0 0 0.0855 -0.9963>}
 dyn_ponytail_02_r {parent=dyn_ponytail_01_r, loc=<5.1311 0.0058 -0.1089>, rot=<0.0015 -0.0086 0.1894 -0.9819>}
 dyn_ponytail_03_r {parent=dyn_ponytail_02_r, loc=<2.225 0 0>, rot=<0.0104 0.0396 0.2545 -0.9662>}
 dyn_ponytail_04_r {parent=dyn_ponytail_03_r, loc=<3.6454 0 0>, rot=<0 0 0.0607 -0.9982>}
 dyn_ponytail_05_r {parent=dyn_ponytail_04_r, loc=<4.1511 0 0>, rot=<0 0 -0.0803 -0.9968>}
 base_ponytail_00_l {parent=head, loc=<-3.4793 -1.5415 -5.1864>, rot=<-0.5372 -0.842 -0.0353 0.0348>}
 dyn_ponytail_00_l {parent=base_ponytail_00_l, loc=<0 0 0>, rot=<0 0 0 1>}
 dyn_ponytail_01_l {parent=dyn_ponytail_00_l, loc=<4.3059 0 0>, rot=<0 0 0.0876 -0.9962>}
 dyn_ponytail_02_l {parent=dyn_ponytail_01_l, loc=<4.8765 0.0008 0.0098>, rot=<-0.0036 0.0183 0.151 -0.9884>}
 dyn_ponytail_03_l {parent=dyn_ponytail_02_l, loc=<2.2934 0 0>, rot=<0 0 0.2226 -0.9749>}
 dyn_ponytail_04_l {parent=dyn_ponytail_03_l, loc=<2.099 0 0>, rot=<0 0 0.1448 -0.9895>}
 dyn_ponytail_05_l {parent=dyn_ponytail_04_l, loc=<4.0638 0 0>, rot=<0 0 -0.078 -0.997>}
 dyn_ponytail_06_l {parent=dyn_ponytail_05_l, loc=<3.0678 0 0>, rot=<0 0 -0.1056 -0.9944>}
 base_hood_up {parent=head, loc=<15.8933 -1.0466 0>, rot=<0.34 0.62 0.62 0.34>}
 dyn_hood_up {parent=base_hood_up, loc=<0 0 0>, rot=<0 0 0 1>}
 base_ear_00_l {parent=dyn_hood_up, loc=<6.3019 -4.2945 -0.3541>, rot=<0.1984 0.5503 0.689 -0.4279>}
 dyn_ear_00_l {parent=base_ear_00_l, loc=<0 0 0>, rot=<0 0 0 1>}
 dyn_ear_01_l {parent=dyn_ear_00_l, loc=<5.8724 0 0>, rot=<-0.0004 0.0046 -0.0769 -0.997>}
 dyn_ear_02_l {parent=dyn_ear_01_l, loc=<5.6372 0 0>, rot=<-0.0038 0.0412 -0.0911 -0.995>}
 dyn_ear_03_l {parent=dyn_ear_02_l, loc=<6.4584 0 0>, rot=<-0.0002 0.035 -0.0062 -0.9994>}
 dyn_ear_04_l {parent=dyn_ear_03_l, loc=<4.39 0 0>, rot=<0.0036 0.053 0.0672 -0.9963>}
 dyn_ear_05_l {parent=dyn_ear_04_l, loc=<5.2274 0 0>, rot=<0 0 0 1>}
 base_ear_00_r {parent=dyn_hood_up, loc=<6.3787 4.2875 0.2236>, rot=<-0.4273 -0.6899 -0.5502 0.197>}
 dyn_ear_00_r {parent=base_ear_00_r, loc=<0 0 0>, rot=<0 0 0 1>}
 dyn_ear_01_r {parent=dyn_ear_00_r, loc=<6.5855 0 0>, rot=<0.0011 -0.0114 -0.098 -0.9951>}
 dyn_ear_02_r {parent=dyn_ear_01_r, loc=<6.4871 0 0>, rot=<0.0046 -0.0525 -0.0867 -0.9948>}
 dyn_ear_03_r {parent=dyn_ear_02_r, loc=<7.8047 0 0>, rot=<-0.0039 -0.0589 0.0658 -0.9961>}
 dyn_ear_04_r {parent=dyn_ear_03_r, loc=<6.7409 0 0>, rot=<0 0 0 1>}
 base_hood_l {parent=head, loc=<-1.3129 -8.3697 -8.7047>, rot=<0.4075 0.5779 0.5779 0.4075>}
 dyn_hood_l {parent=base_hood_l, loc=<0 0 0>, rot=<0 0 0 1>}
 base_hood_r {parent=head, loc=<-1.3129 -8.3697 8.705>, rot=<0.4075 0.5779 0.5779 0.4075>}
 dyn_hood_r {parent=base_hood_r, loc=<0 0 0>, rot=<0 0 0 1>}
 attach_backpack {parent=spine_05, loc=<0 -8.35 0>, rot=<-0.0093 -0.707 0.0093 -0.707>}
 clavicle_l {parent=spine_05, loc=<5.4831 -0.1053 -3.2946>, rot=<0.0465 -0.7289 -0.009 -0.6829>}
 upperarm_l {parent=clavicle_l, loc=<9.8658 -1.2136 -0.4105>, rot=<0.0164 -0.3245 -0.0269 -0.9454>}
 lowerarm_l {parent=upperarm_l, loc=<26.465 -0.8345 0.259>, rot=<0.0036 0.0161 -0.2298 -0.9731>}
 hand_l {parent=lowerarm_l, loc=<22.7355 -0.3203 -0.6105>, rot=<-0.6784 -0.0223 0.0735 -0.7307>}
 weapon_l {parent=hand_l, loc=<-0.1974 -3.4111 0.3811>, rot=<0 0 0 1>}
 hand_attach_l {parent=hand_l, loc=<8.5 -3 0>, rot=<0.7071 0.7071 0 0>}
 thumb_01_l {parent=hand_l, loc=<1.3118 -0.3191 -2.459>, rot=<0.6546 -0.4042 0.0664 -0.6354>}
 thumb_02_l {parent=thumb_01_l, loc=<5.9336 -0.0448 0.0401>, rot=<-0.0037 0.0213 0.298 -0.9543>}
 thumb_03_l {parent=thumb_02_l, loc=<2.2223 0.3682 -0.0907>, rot=<0.0118 -0.013 0.0066 -0.9998>}
 index_metacarpal_l {parent=hand_l, loc=<1.6535 0.0183 -2.96>, rot=<0.0158 -0.0124 -0.0173 -0.9996>}
 index_01_l {parent=index_metacarpal_l, loc=<8.3008 0.0357 -0.0524>, rot=<-0.0044 0.0485 0.0924 -0.9945>}
 index_02_l {parent=index_01_l, loc=<3.3587 -0.3004 0.0452>, rot=<-0.002 0.0027 0.3225 -0.9466>}
 index_03_l {parent=index_02_l, loc=<2.5578 -0.123 -0.0252>, rot=<-0.002 0.0067 0.2558 -0.9667>}
 middle_metacarpal_l {parent=hand_l, loc=<2.1856 -0.1725 -0.5984>, rot=<0.006 0.0168 -0.0329 -0.9993>}
 middle_01_l {parent=middle_metacarpal_l, loc=<7.3275 -0.2447 0.0129>, rot=<-0.0208 0.0042 0.0858 -0.9961>}
 middle_02_l {parent=middle_01_l, loc=<4.0123 0.1401 0.1175>, rot=<-0.0218 0.0258 0.3512 -0.9357>}
 middle_03_l {parent=middle_02_l, loc=<2.6077 0.0534 0.1373>, rot=<0.0082 0.0188 0.1708 -0.9851>}
 ring_metacarpal_l {parent=hand_l, loc=<2.2074 -0.3625 1.3638>, rot=<-0.0284 0.0722 -0.0458 -0.9959>}
 ring_01_l {parent=ring_metacarpal_l, loc=<7.0295 -0.6197 -0.1446>, rot=<-0.0489 -0.0252 0.1892 -0.9804>}
 ring_02_l {parent=ring_01_l, loc=<3.6174 0.4647 -0.1027>, rot=<0 0 0.2806 -0.9598>}
 ring_03_l {parent=ring_02_l, loc=<2.4152 0.114 0.0199>, rot=<0 0 0.1189 -0.9929>}
 pinky_metacarpal_l {parent=hand_l, loc=<1.7616 -0.9753 3.3814>, rot=<-0.0397 0.0814 -0.069 -0.9935>}
 pinky_01_l {parent=pinky_metacarpal_l, loc=<7.1709 -0.4831 -0.1094>, rot=<-0.0654 -0.0035 0.2242 -0.9723>}
 pinky_02_l {parent=pinky_01_l, loc=<2.8274 0.0893 -0.1043>, rot=<0.0037 0.0011 0.2854 -0.9584>}
 pinky_03_l {parent=pinky_02_l, loc=<2.2046 0.0307 -0.0078>, rot=<-0.0002 0.0098 0.1445 -0.9895>}
 lowerarm_twist_02_l {parent=lowerarm_l, loc=<10.6707 0 0>, rot=<0 0 0 1>}
 lowerarm_twist_01_l {parent=lowerarm_l, loc=<18.316 0 0>, rot=<0 0 0 1>}
 upperarm_twist_01_l {parent=upperarm_l, loc=<7.2135 -0.1134 0>, rot=<0 0 0 1>}
 upperarm_twist_02_l {parent=upperarm_l, loc=<19.7998 -0.1134 0>, rot=<0 0 0 1>}
 clavicle_r {parent=spine_05, loc=<5.4831 -0.1053 3.2946>, rot=<0.7289 0.0465 -0.6829 0.009>}
 upperarm_r {parent=clavicle_r, loc=<-9.8658 1.2136 0.4105>, rot=<0.0164 -0.3245 -0.0269 -0.9454>}
 lowerarm_r {parent=upperarm_r, loc=<-26.465 0.8345 -0.259>, rot=<0.0036 0.0161 -0.2298 -0.9731>}
 hand_r {parent=lowerarm_r, loc=<-22.7355 0.3203 0.6106>, rot=<-0.6784 -0.0223 0.0735 -0.7307>}
 weapon_r {parent=hand_r, loc=<0.1974 3.4111 -0.3811>, rot=<0 0 0 1>}
 thumb_01_r {parent=hand_r, loc=<-1.3118 0.3191 2.459>, rot=<0.6546 -0.4042 0.0664 -0.6354>}
 thumb_02_r {parent=thumb_01_r, loc=<-5.9336 0.0448 -0.0401>, rot=<-0.0037 0.0213 0.298 -0.9543>}
 thumb_03_r {parent=thumb_02_r, loc=<-2.2223 -0.3682 0.0907>, rot=<0.0118 -0.013 0.0066 -0.9998>}
 index_metacarpal_r {parent=hand_r, loc=<-1.6535 -0.0183 2.96>, rot=<0.0158 -0.0124 -0.0173 -0.9996>}
 index_01_r {parent=index_metacarpal_r, loc=<-8.3008 -0.0357 0.0524>, rot=<-0.0044 0.0485 0.0924 -0.9945>}
 index_02_r {parent=index_01_r, loc=<-3.3587 0.3004 -0.0452>, rot=<-0.002 0.0027 0.3225 -0.9466>}
 index_03_r {parent=index_02_r, loc=<-2.5578 0.123 0.0252>, rot=<-0.002 0.0067 0.2558 -0.9667>}
 middle_metacarpal_r {parent=hand_r, loc=<-2.1856 0.1725 0.5984>, rot=<0.006 0.0168 -0.0329 -0.9993>}
 middle_01_r {parent=middle_metacarpal_r, loc=<-7.3275 0.2447 -0.0129>, rot=<-0.0208 0.0042 0.0858 -0.9961>}
 middle_02_r {parent=middle_01_r, loc=<-4.0123 -0.1401 -0.1175>, rot=<-0.0218 0.0258 0.3512 -0.9357>}
 middle_03_r {parent=middle_02_r, loc=<-2.6077 -0.0534 -0.1373>, rot=<0.0082 0.0188 0.1708 -0.9851>}
 ring_metacarpal_r {parent=hand_r, loc=<-2.2074 0.3625 -1.3638>, rot=<-0.0284 0.0722 -0.0458 -0.9959>}
 ring_01_r {parent=ring_metacarpal_r, loc=<-7.0295 0.6197 0.1446>, rot=<-0.0489 -0.0252 0.1892 -0.9804>}
 ring_02_r {parent=ring_01_r, loc=<-3.6174 -0.4647 0.1027>, rot=<0 0 0.2806 -0.9598>}
 ring_03_r {parent=ring_02_r, loc=<-2.4152 -0.114 -0.0199>, rot=<0 0 0.1189 -0.9929>}
 pinky_metacarpal_r {parent=hand_r, loc=<-1.7616 0.9753 -3.3814>, rot=<-0.0397 0.0814 -0.069 -0.9935>}
 pinky_01_r {parent=pinky_metacarpal_r, loc=<-7.1709 0.4831 0.1094>, rot=<-0.0654 -0.0035 0.2242 -0.9723>}
 pinky_02_r {parent=pinky_01_r, loc=<-2.8274 -0.0893 0.1043>, rot=<0.0037 0.0011 0.2854 -0.9584>}
 pinky_03_r {parent=pinky_02_r, loc=<-2.2046 -0.0307 0.0078>, rot=<-0.0002 0.0098 0.1445 -0.9895>}
 hand_attach_r {parent=hand_r, loc=<-8.5001 2.9999 0>, rot=<0 0 -0.7071 -0.7071>}
 lowerarm_twist_02_r {parent=lowerarm_r, loc=<-10.6707 0 0>, rot=<0 0 0 -1>}
 lowerarm_twist_01_r {parent=lowerarm_r, loc=<-18.316 0 0>, rot=<0 0 0 -1>}
 upperarm_twist_01_r {parent=upperarm_r, loc=<-7.2135 0.1134 0>, rot=<0 0 0 -1>}
 upperarm_twist_02_r {parent=upperarm_r, loc=<-19.7998 0.1134 0>, rot=<0 0 0 -1>}
 thigh_l {parent=pelvis, loc=<-1.0855 -0.2109 -8.7053>, rot=<0.0009 0.0355 0.0155 -0.9993>}
 calf_l {parent=thigh_l, loc=<-42.5284 -0.0019 -0.7297>, rot=<0.0006 -0.0161 -0.0391 -0.9991>}
 foot_l {parent=calf_l, loc=<-46.7401 -2.5014 -0.592>, rot=<0.0427 -0.0238 0.0353 -0.9982>}
 ball_l {parent=foot_l, loc=<-3.5658 10.8066 -0.1544>, rot=<0.0001 -0.0001 -0.7063 -0.7079>}
 calf_twist_02_l {parent=calf_l, loc=<-19.1546 -1.7943 -0.2638>, rot=<0.0237 0.0095 -0.0035 -0.9997>}
 calf_twist_01_l {parent=calf_l, loc=<-35.8483 -1.9197 -0.5794>, rot=<0.0237 0.0095 -0.0035 -0.9997>}
 thigh_twist_01_l {parent=thigh_l, loc=<-27.952 0 0>, rot=<0.0013 0 0 -1>}
 thigh_r {parent=pelvis, loc=<-1.0855 -0.2109 8.7053>, rot=<0.0355 -0.0009 0.9993 0.0155>}
 calf_r {parent=thigh_r, loc=<42.5284 0.0019 0.7297>, rot=<0.0006 -0.0161 -0.0391 -0.9991>}
 foot_r {parent=calf_r, loc=<46.7401 2.5014 0.592>, rot=<0.0427 -0.0238 0.0353 -0.9982>}
 ball_r {parent=foot_r, loc=<3.5658 -10.8066 0.1544>, rot=<0.0001 -0.0001 -0.7063 -0.7079>}
 calf_twist_02_r {parent=calf_r, loc=<19.1546 1.7943 0.2638>, rot=<0.0237 0.0095 -0.0035 -0.9997>}
 calf_twist_01_r {parent=calf_r, loc=<35.8483 1.9197 0.5794>, rot=<0.0237 0.0095 -0.0035 -0.9997>}
 thigh_twist_01_r {parent=thigh_r, loc=<27.952 0 0>, rot=<0.0013 0 0 -1>}
 attach {parent=root, loc=<0 0 0>, rot=<0 0 0 1>}
 ik_hand_root {parent=root, loc=<0 0 0>, rot=<0 0 0 1>}
 ik_hand_gun {parent=ik_hand_root, loc=<-48.1743 3.3545 104.4381>, rot=<0.6276 0.1309 0.4183 -0.6434>}
 ik_hand_r {parent=ik_hand_gun, loc=<0 0 0>, rot=<0 0 0 1>}
 ik_hand_l {parent=ik_hand_gun, loc=<59.3271 67.6997 34.3524>, rot=<-0.6158 -0.7027 -0.3565 0>}
 ik_foot_root {parent=root, loc=<0 0 0>, rot=<0 0 0 1>}
 ik_foot_l {parent=ik_foot_root, loc=<14.8262 -7.3545 8.296>, rot=<0.0357 0.7045 0.0284 -0.7082>}
 ik_foot_r {parent=ik_foot_root, loc=<-14.8262 -7.3545 8.296>, rot=<0.7082 0.0284 -0.7045 0.0357>}
 attach_fp {parent=root, loc=<0 0 0>, rot=<0 0 0 1>}
 camera_root {parent=root, loc=<0 0 0>, rot=<0 0 0 1>}
 camera {parent=camera_root, loc=<0 0 0>, rot=<0 0 0 1>}
 warp_point_01 {parent=root, loc=<0 0 0>, rot=<0 0 0 1>}
 warp_point_02 {parent=root, loc=<0 0 0>, rot=<0 0 0 1>}
 warp_point_03 {parent=root, loc=<0 0 0>, rot=<0 0 0 1>}
]
