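# Physics asset for the female medium VitalPsych hair part, giving its left and right ponytails capsule bodies for physics simulation.
# Each ponytail is a capsule chain from a small base body on the head, seven dynamic links left and six right, with tight twist and swing limits that widen past the root.
# Head, upper spine and upper arm capsules, joined by free-rotation constraints, form a body proxy, likely for the hair to collide against.
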PhysicsAsset /Game/Characters/Player/Female/Medium/Bodies/F_MED_VitalPsych/Meshes/Parts/F_MED_VitalPsych_FaceAcc_Hair_Physics1
bodies {
 base_ponytail_00_l [{shapeType=Capsule, length=0.505, radius=0.505, shapeName=base_ponytail_00_l_capsule, rotation=<rot -90 0 0>}]
 base_ponytail_00_r [{shapeType=Capsule, length=0.505, radius=0.505, shapeName=base_ponytail_00_r_capsule, rotation=<rot -90 0 0>}]
 dyn_ponytail_00_l [{shapeType=Capsule, length=2.4845, radius=1.5883, shapeName=dyn_ponytail_00_l_capsule, center=<3.3343 0 0>, rotation=<rot -90 0 0>}]
 dyn_ponytail_00_r [{shapeType=Capsule, length=2.2315, radius=1.8827, shapeName=dyn_ponytail_00_r_capsule, center=<3.3302 0 0>, rotation=<rot -90 0 0>}]
 dyn_ponytail_01_l [{shapeType=Capsule, length=2, radius=1.5, shapeName=dyn_ponytail_01_l_capsule, center=<2.7101 0 0>, rotation=<rot -90 0 0>}]
 dyn_ponytail_01_r [{shapeType=Capsule, length=2, radius=1.5, shapeName=dyn_ponytail_01_r_capsule, center=<2.5577 0 0>, rotation=<rot -90 0 0>}]
 dyn_ponytail_02_l [{shapeType=Capsule, length=1.364, radius=1.0528, shapeName=dyn_ponytail_02_l_capsule, center=<0.9929 0 0>, rotation=<rot -90 0 0>}]
 dyn_ponytail_02_r [{shapeType=Capsule, length=1.6512, radius=1.4308, shapeName=dyn_ponytail_02_r_capsule, center=<0.9569 0 0>, rotation=<rot -90 0 0>}]
 dyn_ponytail_03_l [{shapeType=Capsule, length=2.0826, radius=1.2956, shapeName=dyn_ponytail_03_l_capsule, center=<1.198 0 0>, rotation=<rot -90 0 0>}]
 dyn_ponytail_03_r [{shapeType=Capsule, length=1.6802, radius=1.0807, shapeName=dyn_ponytail_03_r_capsule, center=<2.0633 0 0>, rotation=<rot -90 0 0>}]
 dyn_ponytail_04_l [{shapeType=Capsule, length=1.2096, radius=0.8864, shapeName=dyn_ponytail_04_l_capsule, center=<1.8642 0 0>, rotation=<rot -90 0 0>}]
 dyn_ponytail_04_r [{shapeType=Capsule, length=1.619, radius=0.8002, shapeName=dyn_ponytail_04_r_capsule, center=<2.0992 0 0>, rotation=<rot -90 0 0>}]
 dyn_ponytail_05_l [{shapeType=Capsule, length=1.6991, radius=0.5665, shapeName=dyn_ponytail_05_l_capsule, center=<1.1534 0 0>, rotation=<rot -90 0 0>}]
 dyn_ponytail_05_r [{shapeType=Capsule, length=4.5205, radius=0.6315, shapeName=dyn_ponytail_05_r_capsule, center=<3.6161 0 0>, rotation=<rot -90 0 0>}]
 dyn_ponytail_06_l [{shapeType=Capsule, length=3.0948, radius=0.5, shapeName=dyn_ponytail_06_l_capsule, center=<2.2177 0 0>, rotation=<rot -90 0 0>}]
 head [{shapeType=Capsule, length=1, radius=1}]
 spine_03 [{shapeType=Capsule, length=25, radius=8.5, shapeName=spine_03_capsule, center=<0 0.5871 0>, rotation=<rot 0 -90 0>}]
 spine_04 [{shapeType=Capsule, length=21.5, radius=8.255, shapeName=spine_04_capsule, center=<3.6984 1.5693 0>, rotation=<rot 0 -90 0>}]
 spine_05 [{
   shapeType=Capsule
   length=14.4528
   radius=8.4693
   shapeName=spine_05_capsule
   center=<5.742 0.9806 0.0084>
   rotation=<rot 1.9269 -90 -1.9551>
  }]
 upperarm_l [{shapeType=Capsule, length=13.755, radius=4, shapeName=upperarm_r_capsule, center=<11.838 0 0>, rotation=<rot 90 0 0>}]
 upperarm_r [{shapeType=Capsule, length=13.755, radius=4, shapeName=upperarm_r_capsule, center=<-11.838 0 0>, rotation=<rot 90 0 0>}]
}
constraints [
 {
  bone1Name=base_ponytail_00_l
  bone2Name=head
  swing1LimitDegrees=45
  swing1Motion=Free
  swing2LimitDegrees=45
  swing2Motion=Free
  twistLimitDegrees=45
  twistMotion=Free
 }
 {
  bone1Name=base_ponytail_00_r
  bone2Name=head
  swing1LimitDegrees=45
  swing1Motion=Free
  swing2LimitDegrees=45
  swing2Motion=Free
  twistLimitDegrees=45
  twistMotion=Free
 }
 {
  bone1Name=head
  bone2Name=spine_05
  swing1LimitDegrees=45
  swing1Motion=Free
  swing2LimitDegrees=45
  swing2Motion=Free
  twistLimitDegrees=45
  twistMotion=Free
 }
 {
  bone1Name=spine_04
  bone2Name=spine_03
  swing1LimitDegrees=45
  swing1Motion=Free
  swing2LimitDegrees=45
  swing2Motion=Free
  twistLimitDegrees=45
  twistMotion=Free
 }
 {
  bone1Name=spine_05
  bone2Name=spine_04
  swing1LimitDegrees=45
  swing1Motion=Free
  swing2LimitDegrees=45
  swing2Motion=Free
  twistLimitDegrees=45
  twistMotion=Free
 }
 {bone1Name=dyn_ponytail_00_l, bone2Name=base_ponytail_00_l, swing1LimitDegrees=15, swing2LimitDegrees=15, twistLimitDegrees=1}
 {bone1Name=dyn_ponytail_01_l, bone2Name=dyn_ponytail_00_l, swing1LimitDegrees=20, swing2LimitDegrees=15, twistLimitDegrees=1}
 {bone1Name=dyn_ponytail_02_l, bone2Name=dyn_ponytail_01_l, swing1LimitDegrees=25, swing2LimitDegrees=30, twistLimitDegrees=1}
 {bone1Name=dyn_ponytail_03_l, bone2Name=dyn_ponytail_02_l, swing1LimitDegrees=45, swing2LimitDegrees=45, twistLimitDegrees=1}
 {bone1Name=dyn_ponytail_04_l, bone2Name=dyn_ponytail_03_l, swing1LimitDegrees=45, swing2LimitDegrees=45, twistLimitDegrees=1}
 {bone1Name=dyn_ponytail_05_l, bone2Name=dyn_ponytail_04_l, swing1LimitDegrees=30, swing2LimitDegrees=30, twistLimitDegrees=1}
 {bone1Name=dyn_ponytail_06_l, bone2Name=dyn_ponytail_05_l, swing1LimitDegrees=30, swing2LimitDegrees=30, twistLimitDegrees=1}
 {bone1Name=dyn_ponytail_00_r, bone2Name=base_ponytail_00_r, swing1LimitDegrees=15, swing2LimitDegrees=15, twistLimitDegrees=1}
 {bone1Name=dyn_ponytail_01_r, bone2Name=dyn_ponytail_00_r, swing1LimitDegrees=20, swing2LimitDegrees=15, twistLimitDegrees=1}
 {bone1Name=dyn_ponytail_02_r, bone2Name=dyn_ponytail_01_r, swing1LimitDegrees=25, swing2LimitDegrees=30, twistLimitDegrees=1}
 {bone1Name=dyn_ponytail_03_r, bone2Name=dyn_ponytail_02_r, swing1LimitDegrees=30, swing2LimitDegrees=30, twistLimitDegrees=1}
 {bone1Name=dyn_ponytail_04_r, bone2Name=dyn_ponytail_03_r, swing1LimitDegrees=30, swing2LimitDegrees=30, twistLimitDegrees=1}
 {bone1Name=dyn_ponytail_05_r, bone2Name=dyn_ponytail_04_r, swing1LimitDegrees=30, swing2LimitDegrees=30, twistLimitDegrees=1}
 {
  bone1Name=upperarm_r
  bone2Name=spine_05
  swing1LimitDegrees=45
  swing1Motion=Free
  swing2LimitDegrees=45
  swing2Motion=Free
  twistLimitDegrees=45
  twistMotion=Free
 }
 {
  bone1Name=upperarm_l
  bone2Name=spine_05
  swing1LimitDegrees=45
  swing1Motion=Free
  swing2LimitDegrees=45
  swing2Motion=Free
  twistLimitDegrees=45
  twistMotion=Free
 }
]
