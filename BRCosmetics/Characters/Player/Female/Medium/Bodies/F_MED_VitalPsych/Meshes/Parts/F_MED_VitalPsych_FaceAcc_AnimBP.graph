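# Anim Blueprint for the VitalPsych female face accessory, adding secondary motion to ponytails, hood and ears over poses copied from parent and body meshes.
# Constraints and head-turn corrections shape the cached ponytail pose, feeding rigid-body passes selected by front-end/emote state.
# Output adds ear physics with offsets, or trail controllers from the ponytail cache when rigid bodies are disabled.
# Update logic drives gravity, skydive, airborne/crouch and head-rotation inputs; another ponytail branch is unconnected.
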AnimBlueprint /BRCosmetics/Characters/Player/Female/Medium/Bodies/F_MED_VitalPsych/Meshes/Parts/F_MED_VitalPsych_FaceAcc_AnimBP
target_skeleton=/Game/Characters/Player/Female/Medium/Bodies/F_MED_VitalPsych/Meshes/Parts/F_MED_VitalPsych_FaceAcc_Skeleton
graphs [
 AnimGraph {nodes [
   n1 = Get bIsParachuteOpened()
   n2 = Get Skydive() -> (Skydive:bool)
   n3 = RigidBody!(
    ComponentPose
    Alpha=SelectFloat(A=0.25, B=1, bPickA=OR Boolean(A=@n2, B=@n1))
    OverrideWorldGravity:Vector=Get GravityOverride()
    ExternalForce
   ) -> (Pose) {
    bClampLinearTranslationLimitToRefPose=true
    bOverrideWorldGravity=true
    overridePhysicsAsset=/Game/Characters/Player/Female/Medium/Bodies/F_MED_VitalPsych/Meshes/Parts/F_MED_VitalPsych_FaceAcc_Hair_Physics1
    simulationSpace=BaseBoneSpace
    alphaBoolBlend {blendOption=Linear}
    alphaScaleBiasClamp {bInterpResult=true, clampMin=0.25, interpSpeedDecreasing=4, interpSpeedIncreasing=4}
    baseBoneRef {boneName=spine_05}
    componentAppliedLinearAccClamp=<3000 3000 3000>
    componentLinearAccScale=<0.22 0.22 0.5>
    componentLinearVelScale=<0.22 0.22 0.5>
    overrideWorldGravity {z=-980}
    simSpaceSettings {worldAlpha=0.15}
   }
   n4 = "To Float (Boolean)"(InBool=Get Skydive())
   n5 = "Transform (Modify) Bone - Bone: dyn_ponytail_00_r"!(ComponentPose=@n3, Alpha=@n4, Rotation="-28.743,-4.662,9.625") -> (Pose) {
    rotationMode=BMM_Additive
    alphaBoolBlend {blendOption=Linear}
    alphaScaleBiasClamp {bInterpResult=true}
    boneToModify {boneName=dyn_ponytail_00_r}
    rotation=<rot -28.743 -4.662 9.625>
   }
   n6 = "Transform (Modify) Bone - Bone: dyn_ponytail_00_l"!(ComponentPose=@n5, Alpha=@n4, Rotation="28.764,4.516,9.320") -> (Pose) {
    rotationMode=BMM_Additive
    alphaBoolBlend {blendOption=Linear}
    alphaScaleBiasClamp {bInterpResult=true}
    boneToModify {boneName=dyn_ponytail_00_l}
    rotation=<rot 28.764 4.516 9.32>
   }
   n7 = "Transform (Modify) Bone - Bone: dyn_ponytail_02_r"!(ComponentPose=@n6, Alpha=@n4, Rotation="-47.328,0.000,-0.000") -> (Pose) {
    rotationMode=BMM_Additive
    alphaBoolBlend {blendOption=Linear}
    alphaScaleBiasClamp {bInterpResult=true}
    boneToModify {boneName=dyn_ponytail_02_r}
    rotation {pitch=-47.328}
   }
   n8 = "Transform (Modify) Bone - Bone: dyn_ponytail_02_l"!(ComponentPose=@n7, Alpha=@n4, Rotation="66.203,0.000,0.000") -> (Pose) {
    rotationMode=BMM_Additive
    alphaBoolBlend {blendOption=Linear}
    alphaScaleBiasClamp {bInterpResult=true}
    boneToModify {boneName=dyn_ponytail_02_l}
    rotation {pitch=66.203}
   }
   n9 = "Transform (Modify) Bone - Bone: dyn_ponytail_03_r"!(ComponentPose=@n8, Alpha=@n4, Rotation="-26.845,0.000,-0.000") -> (Pose) {
    rotationMode=BMM_Additive
    alphaBoolBlend {blendOption=Linear}
    alphaScaleBiasClamp {bInterpResult=true}
    boneToModify {boneName=dyn_ponytail_03_r}
    rotation {pitch=-26.845}
   }
   n10 = "Transform (Modify) Bone - Bone: dyn_ponytail_03_l"!(ComponentPose=@n9, Alpha=@n4, Rotation="11.020,0.000,0.000") -> (Pose) {
    rotationMode=BMM_Additive
    alphaBoolBlend {blendOption=Linear}
    alphaScaleBiasClamp {bInterpResult=true}
    boneToModify {boneName=dyn_ponytail_03_l}
    rotation {pitch=11.02}
   }
   n11 = "Transform (Modify) Bone - Bone: dyn_ponytail_04_l"!(ComponentPose=@n10, Alpha=@n4, Rotation="28.479,0.000,0.000") -> (Pose) {
    rotationMode=BMM_Additive
    alphaBoolBlend {blendOption=Linear}
    alphaScaleBiasClamp {bInterpResult=true}
    boneToModify {boneName=dyn_ponytail_04_l}
    rotation {pitch=28.479}
   }
   n12 = "Transform (Modify) Bone - Bone: dyn_ponytail_04_r"!(ComponentPose=@n11, Alpha=@n4, Rotation="-35.015,0.000,-0.000") -> (Pose) {
    rotationMode=BMM_Additive
    alphaBoolBlend {blendOption=Linear}
    alphaScaleBiasClamp {bInterpResult=true}
    boneToModify {boneName=dyn_ponytail_04_r}
    rotation {pitch=-35.015}
   }
   n13 = Control Rig() {
    alphaBoolBlend {blendOption=Linear}
    controlRigAssetReference {blueprintRigClass=/Game/Characters/Player/Female/Medium/Bodies/F_MED_VitalPsych/Meshes/Parts/F_MED_VitalPsych_FaceAcc_CtrlRig.F_MED_VitalPsych_FaceAcc_CtrlRig_C}
   }
   n14 = Trail controller - Bone: dyn_ponytail_06_l!(ComponentPose=Local To Component(LocalPose=@n13), Alpha=0.25, RelaxationSpeedScale:float=2, FakeVelocity:Vector="0,0,0") -> (Pose) {
    bActorSpaceFakeVel=true
    bShowTrailLocation=true
    bUsePlanarLimit=true
    chainLength=5
    debugLifeTime=0.1
    maxDeltaTime=0.0333
    relaxationSpeedScale=1.2
    alphaScaleBiasClamp {clampMax=0.25, interpSpeedDecreasing=25, interpSpeedIncreasing=25}
    baseJoint {boneName=root}
    rotationLimits [{limitMax=<180 180 180>}, {limitMax=<180 180 180>}, {limitMax=<180 180 180>}, {limitMax=<180 180 180>}, {limitMax=<180 180 180>}]
    rotationOffsets [<0 0 0>, <0 0 0>, <0 0 0>, <0 0 0>, <0 0 0>]
    trailBone {boneName=dyn_ponytail_06_l}
    trailRelaxationSpeed {editorCurveData {keys [<at 0 15 RCIM_Linear>,<at 1 6 RCIM_Linear>]}}
   }
   n15 = Trail controller - Bone: dyn_ponytail_05_r!(ComponentPose=@n14, Alpha=0.25, RelaxationSpeedScale:float=2, FakeVelocity:Vector="0,0,0") -> (Pose) {
    bActorSpaceFakeVel=true
    bShowTrailLocation=true
    bUsePlanarLimit=true
    chainLength=5
    debugLifeTime=0.1
    maxDeltaTime=0.0333
    relaxationSpeedScale=1.2
    alphaScaleBiasClamp {clampMax=0.25, interpSpeedDecreasing=25, interpSpeedIncreasing=25}
    baseJoint {boneName=root}
    rotationLimits [{limitMax=<180 180 180>}, {limitMax=<180 180 180>}, {limitMax=<180 180 180>}, {limitMax=<180 180 180>}, {limitMax=<180 180 180>}]
    rotationOffsets [<0 0 0>, <0 0 0>, <0 0 0>, <0 0 0>, <0 0 0>]
    trailBone {boneName=dyn_ponytail_05_r}
    trailRelaxationSpeed {editorCurveData {keys [<at 0 15 RCIM_Linear>,<at 1 6 RCIM_Linear>]}}
   }
   Blend Poses by bool(BlendPose_0=Component To Local(ComponentPose=@n12), BlendPose_1=Component To Local(ComponentPose=@n15), bActiveValue=Get bRigidBodyEnabled()) {blendType=Linear}
   n16 = Get Skydive()
   n17 = RigidBody(
    ComponentPose=Local To Component(LocalPose=Use cached pose 'Hood_Dyn_Cache'())
    Alpha=SelectFloat(A=0.4, B=0.75, bPickA=@n16)
    OverrideWorldGravity="0,0,-980"
   ) {
    bClampLinearTranslationLimitToRefPose=true
    bOverrideWorldGravity=true
    overridePhysicsAsset=&/F_MED_VitalPsych_FaceAcc_Ears_Physics
    simulationSpace=BaseBoneSpace
    alphaBoolBlend {blendOption=Linear}
    alphaScaleBiasClamp {clampMin=0.25, interpSpeedDecreasing=4, interpSpeedIncreasing=4}
    baseBoneRef {boneName=spine_05}
    componentAppliedLinearAccClamp=<3000 3000 3000>
    componentLinearAccScale=<0.22 0.22 0.5>
    componentLinearVelScale=<0.22 0.22 0.5>
    overrideWorldGravity {z=-980}
    simSpaceSettings {worldAlpha=0.25}
   }
   n18 = "Transform (Modify) Bone - Bone: dyn_ear_00_r"!(ComponentPose=@n17, bAlphaBoolEnabled:bool=True, Rotation="-10,0,0") -> (Pose) {
    alphaInputType=Bool
    rotationMode=BMM_Additive
    alphaBoolBlend {blendOption=Linear}
    boneToModify {boneName=dyn_ear_00_r}
    rotation {pitch=-10}
   }
   n19 = "Transform (Modify) Bone - Bone: dyn_ear_01_r"!(ComponentPose=@n18, bAlphaBoolEnabled:bool=True, Rotation="-15,0,0") -> (Pose) {
    alphaInputType=Bool
    rotationMode=BMM_Additive
    alphaBoolBlend {blendOption=Linear}
    boneToModify {boneName=dyn_ear_01_r}
    rotation {pitch=-15}
   }
   n20 = "Transform (Modify) Bone - Bone: dyn_ear_02_r"!(ComponentPose=@n19, bAlphaBoolEnabled:bool=True, Rotation="-5,0,0") -> (Pose) {
    alphaInputType=Bool
    rotationMode=BMM_Additive
    alphaBoolBlend {blendOption=Linear}
    boneToModify {boneName=dyn_ear_02_r}
    rotation {pitch=-5}
   }
   n21 = "Transform (Modify) Bone - Bone: dyn_ear_03_r"!(ComponentPose=@n20, bAlphaBoolEnabled:bool=True, Rotation="-5,0,0") -> (Pose) {
    alphaInputType=Bool
    rotationMode=BMM_Additive
    alphaBoolBlend {blendOption=Linear}
    boneToModify {boneName=dyn_ear_03_r}
    rotation {pitch=-5}
   }
   n22 = "Transform (Modify) Bone - Bone: dyn_ear_00_l"!(ComponentPose=@n21, bAlphaBoolEnabled:bool=True, Rotation="5.000,0.000,0.000") -> (Pose) {
    alphaInputType=Bool
    rotationMode=BMM_Additive
    alphaBoolBlend {blendOption=Linear}
    boneToModify {boneName=dyn_ear_00_l}
    rotation {pitch=5}
   }
   n23 = "Transform (Modify) Bone - Bone: dyn_ear_01_l"!(ComponentPose=@n22, bAlphaBoolEnabled:bool=True, Rotation="10.000,0.000,0.000") -> (Pose) {
    alphaInputType=Bool
    rotationMode=BMM_Additive
    alphaBoolBlend {blendOption=Linear}
    boneToModify {boneName=dyn_ear_01_l}
    rotation {pitch=10}
   }
   n24 = "Transform (Modify) Bone - Bone: dyn_ear_02_l"!(ComponentPose=@n23, bAlphaBoolEnabled:bool=True, Rotation="10.000,0.000,0.000") -> (Pose) {
    alphaInputType=Bool
    rotationMode=BMM_Additive
    alphaBoolBlend {blendOption=Linear}
    boneToModify {boneName=dyn_ear_02_l}
    rotation {pitch=10}
   }
   n25 = "Transform (Modify) Bone - Bone: dyn_ear_03_l"!(ComponentPose=@n24, bAlphaBoolEnabled:bool=True, Rotation="10.000,0.000,0.000") -> (Pose) {
    alphaInputType=Bool
    rotationMode=BMM_Additive
    alphaBoolBlend {blendOption=Linear}
    boneToModify {boneName=dyn_ear_03_l}
    rotation {pitch=10}
   }
   n26 = "Transform (Modify) Bone - Bone: dyn_ear_04_l"!(ComponentPose=@n25, bAlphaBoolEnabled:bool=True, Rotation="5.000,0.000,0.000") -> (Pose) {
    alphaInputType=Bool
    rotationMode=BMM_Additive
    alphaBoolBlend {blendOption=Linear}
    boneToModify {boneName=dyn_ear_04_l}
    rotation {pitch=5}
   }
   n27 = Get Is Jumping Falling() -> (Is Jumping Falling:double)
   n28 = Trail controller - Bone: dyn_ear_05_l!(
    ComponentPose=Local To Component(LocalPose=Use cached pose 'LongHair_Dyn_Cache'())
    Alpha=@n27
    RelaxationSpeedScale:float=3
    FakeVelocity:Vector="0,0,0"
   ) -> (Pose) {
    bActorSpaceFakeVel=true
    bShowTrailLocation=true
    bUsePlanarLimit=true
    chainLength=5
    debugLifeTime=0.1
    maxDeltaTime=0.0333
    relaxationSpeedScale=1.2
    alphaScaleBiasClamp {bClampResult=true, bInterpResult=true, clampMax=0.45, clampMin=0.15, interpSpeedDecreasing=50, interpSpeedIncreasing=5}
    baseJoint {boneName=root}
    rotationLimits [{limitMax=<180 180 180>}, {limitMax=<180 180 180>}, {limitMax=<180 180 180>}, {limitMax=<180 180 180>}, {limitMax=<180 180 180>}]
    rotationOffsets [<0 0 0>, <0 0 0>, <0 0 0>, <0 0 0>, <0 0 0>]
    trailBone {boneName=dyn_ear_05_l}
    trailRelaxationSpeed {editorCurveData {keys [<at 0 15 RCIM_Linear>,<at 1 6 RCIM_Linear>]}}
   }
   n29 = Trail controller - Bone: dyn_ear_04_r!(ComponentPose=@n28, Alpha=@n27, RelaxationSpeedScale:float=4, FakeVelocity:Vector="0,0,0") -> (Pose) {
    bActorSpaceFakeVel=true
    bShowTrailLocation=true
    bUsePlanarLimit=true
    chainLength=4
    debugLifeTime=0.1
    maxDeltaTime=0.0333
    relaxationSpeedScale=1.2
    alphaScaleBiasClamp {bClampResult=true, bInterpResult=true, clampMax=0.45, clampMin=0.15, interpSpeedDecreasing=50, interpSpeedIncreasing=5}
    baseJoint {boneName=root}
    rotationLimits [{limitMax=<180 180 180>}, {limitMax=<180 180 180>}, {limitMax=<180 180 180>}, {limitMax=<180 180 180>}]
    rotationOffsets [<0 0 0>, <0 0 0>, <0 0 0>, <0 0 0>]
    trailBone {boneName=dyn_ear_04_r}
    trailRelaxationSpeed {editorCurveData {keys [<at 0 15 RCIM_Linear>,<at 1 6 RCIM_Linear>]}}
   }
   Output Pose(Result=Blend Poses by bool(BlendPose_0=Component To Local(ComponentPose=@n26), BlendPose_1=Component To Local(ComponentPose=@n29), bActiveValue=Get bRigidBodyEnabled()) {blendType=Linear}) {layerGroup=None}
   n30 = "Constraint - base_ponytail_00_l (PPP)"!(
    ComponentPose=Local To Component(LocalPose=Use cached pose 'InputAnimation'())
    Alpha=1
    ConstraintWeights_0=0.6623
    ConstraintWeights_1:float=0.2316
    ConstraintWeights_2:float=0.1061
   ) -> (Pose) {
    alphaBoolBlend {blendOption=Linear}
    boneToModify {boneName=base_ponytail_00_l}
    constraintSetup [{targetBone {boneName=head}}, {targetBone {boneName=neck_02}}, {targetBone {boneName=neck_01}}]
    constraintWeights [1,1,1]
   }
   n31 = "Constraint - dyn_ponytail_01_l (PPP)"!(ComponentPose=@n30, Alpha=1, ConstraintWeights_0=0.6341, ConstraintWeights_1:float=0.3371, ConstraintWeights_2:float=0.0288) -> (Pose) {
    alphaBoolBlend {blendOption=Linear}
    boneToModify {boneName=dyn_ponytail_01_l}
    constraintSetup [{targetBone {boneName=head}}, {targetBone {boneName=neck_02}}, {targetBone {boneName=neck_01}}]
    constraintWeights [1,1,1]
   }
   n32 = "Constraint - dyn_ponytail_01_l (PPPPP)"!(
    ComponentPose=@n31
    Alpha=0.5
    ConstraintWeights_0=0.7156
    ConstraintWeights_1:float=0.2634
    ConstraintWeights_2:float=0.0162
    ConstraintWeights_3:float=0.0032
    ConstraintWeights_4:float=0.0016
   ) -> (Pose) {
    alphaBoolBlend {blendOption=Linear}
    boneToModify {boneName=dyn_ponytail_01_l}
    constraintSetup [
     {targetBone {boneName=spine_05}}
     {targetBone {boneName=clavicle_l}}
     {targetBone {boneName=spine_04}}
     {targetBone {boneName=upperarm_twist_01_l}}
     {targetBone {boneName=neck_02}}
    ]
    constraintWeights [1,1,1,1,1]
   }
   n33 = "Constraint - dyn_ponytail_02_l (PPP)"!(ComponentPose=@n32, Alpha=0.5, ConstraintWeights_0=0.7888, ConstraintWeights_1:float=0.1818, ConstraintWeights_2:float=0.0294) -> (Pose) {
    alphaBoolBlend {blendOption=Linear}
    boneToModify {boneName=dyn_ponytail_02_l}
    constraintSetup [{targetBone {boneName=spine_05}}, {targetBone {boneName=clavicle_l}}, {targetBone {boneName=spine_04}}]
    constraintWeights [1,1,1]
   }
   n34 = "Constraint - base_ponytail_00_r (PPP)"!(ComponentPose=@n33, Alpha=1, ConstraintWeights_0=0.6623, ConstraintWeights_1:float=0.2316, ConstraintWeights_2:float=0.1061) -> (Pose) {
    alphaBoolBlend {blendOption=Linear}
    boneToModify {boneName=base_ponytail_00_r}
    constraintSetup [{targetBone {boneName=head}}, {targetBone {boneName=neck_02}}, {targetBone {boneName=neck_01}}]
    constraintWeights [1,1,1]
   }
   n35 = "Constraint - dyn_ponytail_01_r (PPP)"!(ComponentPose=@n34, Alpha=1, ConstraintWeights_0=0.6341, ConstraintWeights_1:float=0.3371, ConstraintWeights_2:float=0.0288) -> (Pose) {
    alphaBoolBlend {blendOption=Linear}
    boneToModify {boneName=dyn_ponytail_01_r}
    constraintSetup [{targetBone {boneName=head}}, {targetBone {boneName=neck_02}}, {targetBone {boneName=neck_01}}]
    constraintWeights [1,1,1]
   }
   n36 = "Constraint - dyn_ponytail_01_r (PPPPP)"!(
    ComponentPose=@n35
    Alpha=0.35
    ConstraintWeights_0=0.7156
    ConstraintWeights_1:float=0.2634
    ConstraintWeights_2:float=0.0162
    ConstraintWeights_3:float=0.0032
    ConstraintWeights_4:float=0.0016
   ) -> (Pose) {
    alphaBoolBlend {blendOption=Linear}
    boneToModify {boneName=dyn_ponytail_01_r}
    constraintSetup [
     {targetBone {boneName=spine_05}}
     {targetBone {boneName=clavicle_r}}
     {targetBone {boneName=spine_04}}
     {targetBone {boneName=upperarm_twist_01_r}}
     {targetBone {boneName=neck_02}}
    ]
    constraintWeights [1,1,1,1,1]
   }
   n37 = "Constraint - dyn_ponytail_02_r (PPPP)"!(
    ComponentPose=@n36
    Alpha=0.5
    ConstraintWeights_0=0.8527
    ConstraintWeights_1:float=0.1265
    ConstraintWeights_2:float=0.0208
    ConstraintWeights_3:float=0.3
   ) -> (Pose) {
    alphaBoolBlend {blendOption=Linear}
    boneToModify {boneName=dyn_ponytail_02_r}
    constraintSetup [{targetBone {boneName=spine_05}}, {targetBone {boneName=clavicle_r}}, {targetBone {boneName=spine_04}}, {targetBone {boneName=head}}]
    constraintWeights [1,1,1,1]
   }
   n38 = "Transform (Modify) Bone - Bone: dyn_ponytail_00_l"!(ComponentPose=@n37, Alpha=1, Translation:Vector="0,0,0", Rotation="-5.000,-15.000,0.000", Scale:Vector="1,1,1") -> (Pose) {
    rotationMode=BMM_Additive
    rotationSpace=BCS_BoneSpace
    alphaBoolBlend {blendOption=Linear}
    boneToModify {boneName=dyn_ponytail_00_l}
    rotation=<rot -5 -14.9999 0>
   }
   n39 = "Transform (Modify) Bone - Bone: dyn_ponytail_00_r"!(ComponentPose=@n38, Alpha=1, Translation:Vector="0,0,0", Rotation="10.000,-15.000,0.000", Scale:Vector="1,1,1") -> (Pose) {
    rotationMode=BMM_Additive
    rotationSpace=BCS_BoneSpace
    alphaBoolBlend {blendOption=Linear}
    boneToModify {boneName=dyn_ponytail_00_r}
    rotation=<rot 10 -15 0>
   }
   n40 = "Transform (Modify) Bone - Bone: dyn_ponytail_01_l"(ComponentPose=@n39, Alpha=1, Translation="0,0,0", Rotation="0,-25,0", Scale="1,1,1") {
    rotationMode=BMM_Additive
    rotationSpace=BCS_BoneSpace
    alphaBoolBlend {blendOption=Linear}
    boneToModify {boneName=dyn_ponytail_01_l}
    rotation {yaw=-25}
   }
   n41 = "Transform (Modify) Bone - Bone: dyn_ponytail_01_r"(ComponentPose=@n40, Alpha=1, Translation="0,0,0", Rotation="0,-25,0", Scale="1,1,1") {
    rotationMode=BMM_Additive
    rotationSpace=BCS_BoneSpace
    alphaBoolBlend {blendOption=Linear}
    boneToModify {boneName=dyn_ponytail_01_r}
    rotation {yaw=-25}
   }
   n42 = "Transform (Modify) Bone - Bone: dyn_ponytail_04_r"!(ComponentPose=@n41, Alpha=1, Scale:Vector="0,0,0") -> (Pose) {
    rotationSpace=BCS_BoneSpace
    scaleMode=BMM_Additive
    alphaBoolBlend {blendOption=Linear}
    boneToModify {boneName=dyn_ponytail_04_r}
    scale=<0 0 0>
   }
   n43 = "Transform (Modify) Bone - Bone: dyn_ponytail_04_l"!(ComponentPose=@n42, Alpha=1, Scale:Vector="0,0,0") -> (Pose) {
    rotationSpace=BCS_BoneSpace
    scaleMode=BMM_Additive
    alphaBoolBlend {blendOption=Linear}
    boneToModify {boneName=dyn_ponytail_04_l}
    scale=<0 0 0>
   }
   n44 = Get HeadRotate_R() -> (HeadRotate_R:double)
   n45 = "Transform (Modify) Bone - Bone: base_ponytail_00_r"(ComponentPose=@n43, Alpha=@n44, Translation="0,0,0", Rotation="0,0,45", Scale="1,1,1") {
    rotationMode=BMM_Additive
    rotationSpace=BCS_BoneSpace
    alphaBoolBlend {blendOption=Linear}
    boneToModify {boneName=base_ponytail_00_r}
    rotation {roll=45}
   }
   n46 = "Transform (Modify) Bone - Bone: base_ponytail_00_l"(ComponentPose=@n45, Alpha=@n44, Translation="0,0,0", Rotation="0,-45,20", Scale="1,1,1") {
    rotationMode=BMM_Additive
    rotationSpace=BCS_BoneSpace
    alphaBoolBlend {blendOption=Linear}
    boneToModify {boneName=base_ponytail_00_l}
    rotation {roll=20, yaw=-45}
   }
   Save cached pose 'LongHair_Dyn_Cache'(Pose=Component To Local(ComponentPose=@n46)) {cacheName=LongHair_Dyn_Cache}
   n47 = RigidBody(ComponentPose=Local To Component(LocalPose=Use cached pose 'LongHair_Dyn_Cache'()), Alpha=0.5, OverrideWorldGravity="0,0,-800") {
    bClampLinearTranslationLimitToRefPose=true
    bOverrideWorldGravity=true
    overridePhysicsAsset=&/F_MED_VitalPsych_FaceAcc_Physics
    simulationSpace=BaseBoneSpace
    alphaBoolBlend {blendOption=Linear}
    alphaScaleBiasClamp {clampMin=0.25, interpSpeedDecreasing=4, interpSpeedIncreasing=4}
    baseBoneRef {boneName=spine_05}
    componentAppliedLinearAccClamp=<3000 3000 3000>
    componentLinearAccScale=<0.22 0.22 0.5>
    componentLinearVelScale=<0.22 0.22 0.5>
    overrideWorldGravity {z=-980}
    simSpaceSettings {worldAlpha=0.25}
   }
   n48 = "Transform (Modify) Bone - Bone: base_hood_l"(ComponentPose=@n47, Alpha=1, Translation="-2,7,0", Rotation="0,0,0", Scale="1,1,1") {
    scaleMode=BMM_Additive
    scaleSpace=BCS_BoneSpace
    translationMode=BMM_Additive
    translationSpace=BCS_BoneSpace
    alphaBoolBlend {blendOption=Linear}
    boneToModify {boneName=base_hood_l}
    translation=<-2 7>
   }
   n49 = "Transform (Modify) Bone - Bone: base_hood_r"(ComponentPose=@n48, Alpha=1, Translation="0,-4,0", Rotation="0,0,0", Scale="1,1,1") {
    scaleMode=BMM_Additive
    scaleSpace=BCS_BoneSpace
    translationMode=BMM_Additive
    translationSpace=BCS_BoneSpace
    alphaBoolBlend {blendOption=Linear}
    boneToModify {boneName=base_hood_r}
    translation {y=-4}
   }
   n50 = "Transform (Modify) Bone - Bone: base_hair_l"(ComponentPose=@n49, Alpha=1, Translation="0,0,0", Rotation="0,20,0", Scale="1,1,1") {
    rotationMode=BMM_Additive
    rotationSpace=BCS_BoneSpace
    scaleMode=BMM_Additive
    scaleSpace=BCS_BoneSpace
    translationMode=BMM_Additive
    translationSpace=BCS_BoneSpace
    alphaBoolBlend {blendOption=Linear}
    boneToModify {boneName=base_hair_l}
    rotation {yaw=20}
   }
   n51 = Get Skydive()
   n52 = RigidBody(
    ComponentPose=Local To Component(LocalPose=Use cached pose 'LongHair_Dyn_Cache'())
    Alpha=SelectFloat(A=0.25, B=1, bPickA=@n51)
    OverrideWorldGravity=SelectVector(A=Get GravityOverride(), B="0,0,-600", bPickA=@n51)
   ) {
    bClampLinearTranslationLimitToRefPose=true
    bOverrideWorldGravity=true
    overridePhysicsAsset=&F_MED_VitalPsych_FaceAcc_Physics
    simulationSpace=BaseBoneSpace
    alphaBoolBlend {blendOption=Linear}
    alphaScaleBiasClamp {bInterpResult=true, clampMin=0.25, interpSpeedDecreasing=4, interpSpeedIncreasing=4}
    baseBoneRef {boneName=spine_05}
    componentAppliedLinearAccClamp=<3000 3000 3000>
    componentLinearAccScale=<0.22 0.22 0.5>
    componentLinearVelScale=<0.22 0.22 0.5>
    overrideWorldGravity {z=-980}
    simSpaceSettings {worldAlpha=0.25}
   }
   Save cached pose 'Hood_Dyn_Cache'(Pose=Blend Poses by bool(
     BlendPose_0=Component To Local(ComponentPose=@n50)
     BlendPose_1=Component To Local(ComponentPose=@n52)
     bActiveValue=Get FrontEnd_Emote_bool() -> (FrontEnd_Emote_bool:bool)
    ) {blendType=Linear}) {cacheName=Hood_Dyn_Cache}
   n53 = Get Body_Mesh()
   Save cached pose 'InputAnimation'(Pose=Layered blend per bone(BasePose=Copy Pose From Mesh() {bCopyCurves=true, bUseAttachedParent=true}, BlendPoses_0=Copy Pose From Mesh(SourceMeshComponent=@n53)) {layerSetup [{branchFilters [{boneName=upperarm_l}, {boneName=upperarm_r}]}]}) {cacheName=InputAnimation}
   SelectVector(A=Get GravityOverride(), B="0,0,400", bPickA=@n16)
   Get HeadRotate_L() -> (HeadRotate_L:double)
  ]}
 EventGraph {nodes [
   Event BlueprintUpdateAnimation()
   > n1 = Gravity Override State Function DEPRECATED(
    CharacterPartAnimInstance=Self-Reference()
    GameTimeInSeconds=GetGameTimeInSeconds()
    GravityOverrideParams=Get GravityOverrideParams() -> (GravityOverrideParams:Gravity Override Params Struct Structure)
   )
   > Set GravityOverride(GravityOverride=@n1)
   > Set Skydive(Skydive:bool=OR Boolean(A=Get bIsSkydiveDiveMode(), B=Get bIsSkydiveFloating())) -> (Output_Get:bool)
   n2 = OR Boolean!(A=Get bIsJumping(), B=Get bIsFalling(), C:bool=Get bIsCrouching()) -> (ReturnValue)
   > Set Is Jumping Falling(Is Jumping Falling:double="To Float (Boolean)"(InBool=NOT Boolean(A=@n2))) -> (Output_Get:double)
   > Set FrontEnd_Emote_bool(FrontEnd_Emote_bool:bool=OR Boolean!(A=Get bIsInFrontEnd(), B=Get bIsPlayingEmote()) -> (ReturnValue)) -> (Output_Get:bool)
   > Set HeadRotate_L(HeadRotate_L:double=GetAlphaFromReferenceTransformDelta(
     SkeletalMeshComponent=Get Skel Mesh()
     BoneName=spine_05
     BaseName=head
     DeltaComponentType=EulerY
     InputRange=(X=0,Y=-45)
     TwistAxis="0,0,1"
    )) -> (Output_Get:double)
   > Set HeadRotate_R(HeadRotate_R:double=GetAlphaFromReferenceTransformDelta(SkeletalMeshComponent=Get Skel Mesh(), BoneName=spine_05, BaseName=head, DeltaComponentType=EulerY, InputRange=(X=0,Y=45), TwistAxis="0,0,1")) -> (Output_Get:double)
   Event BlueprintInitializeAnimation()
   > Set Skel Mesh(Skel Mesh=GetOwnerSkeletalMeshComponent())
   Event OnCharacterCustomizationLoaded()
   > n3 = Cast To PlayerPawn_Athena(Object=TryGetPawnOwner())
   > Set Body_Mesh(Body_Mesh=GetSkeletalMeshForPartType(self=@n3))
  ]}
]
variables [
 Body_Mesh {type=SkeletalMeshComponent}
 FrontEnd_Emote_bool {type=boolean}
 GravityOverride {type=Vector}
 GravityOverrideParams {type=GravityOverrideParamsStruct}
 HeadRotate_L {type=number}
 HeadRotate_R {type=number}
 Is Jumping Falling {type=number}
 Skel Mesh {type=SkeletalMeshComponent}
 Skydive {type=boolean}
]
